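# Niagara system for the IlluminateLasso emote's prop spawn, throwing short-lived sparkle flares along a rope-shaped path.
# Zero-size rope_shape particles live briefly and follow curve-driven X/Z offsets, acting as invisible guides.
# Flares spawn from those guides through a particle attribute reader, then drift with random velocity, curl noise and rising gravity while scaling and fading.
# Flare colors come from a random three-color palette multiplied by a system time-of-day value.
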NiagaraSystem /BRCosmetics/Animation/Game/MainPlayer/Emotes/IlluminateLasso/FX/NS_Emote_IlluminateLasso_PropSpawn
dependencies {
 usedDataInterfaces [
  &NiagaraDataInterfaceParticleRead
  &NiagaraDataInterfaceArrayColor
  &NiagaraDataInterfaceCurve
  &NiagaraDataInterfaceSkeletalMesh
 ]
 usedDynamicInputs [&FNTimeOfDayColor]
 usedModules [
  &self:SystemSpawnScript.SystemScriptSource.SystemScriptGraph.NiagaraNodeAssignment_0.SetVariables_#1
  &SystemState
  &self:SystemSpawnScript.SystemScriptSource.SystemScriptGraph.NiagaraNodeAssignment_1.SetVariables_#2
 ]
}
emitters {
 Flares {
  emitterUpdateScript [
   &EmitterState(Life Cycle Mode=::NewEnumerator1, Loop Behavior=::NewEnumerator1, Loop Duration=1, UseLoopDelay=false)
   &SpawnBurst_Instantaneous(Spawn Count=35, Use Spawn Probability=false, UseLoopCountLimit=false) {enabled=false}
   &SpawnRate(SpawnRate=8, Use Spawn Probability=false) {enabled=false}
   &SpawnParticlesFromOtherEmitter(
    Attribute Reader=DataInterface() {propertyValues {EmitterBinding {bindingMode=Other, emitterName=rope_shape}}}
    Spawn Rate=500
    Use Maximum Age=false
    Use Minimum Age=false
   )
  ]
  particleSpawnScript [
   &InitializeParticle(
    Color=&MultiplyLinearColor(
     Linear Color=&SelectLinearColorFromArray(
      Array Sampling Mode=::NewEnumerator0
      Direct Array Index=?
      Color Selection Array=DataInterface() {propertyValues {ColorData [<rgba 2.2045 2.4034 3 1>, <rgba 6 2.3575 0 1>, <rgba 3 0.1101 0 1>]}}
      Random Seed=0
     )
     Scale Factor=&MakeVector4_FromLinearColor(LinearColor=$System.TOD)
    )
    Lifetime Max=0.35
    Lifetime Min=0.15
    Lifetime Mode=::NewEnumerator1
    Sprite Rotation Mode=::NewEnumerator2
    Sprite Size Mode=::NewEnumerator1
    Uniform Sprite Size Max=45
    Uniform Sprite Size Min=25
    UsePositionOffset=false
    Write Material Random=true
   )
   &PartitionParticles(Number Of Partitions=1) {enabled=false}
   &ShapeLocation(MaskSpawnGroup=false, Override Seed=false, Sphere Radius=25, UseOwnerScale=true)
   &DynamicMaterialParameters(
    Index 0 Param 1=&RandomRangeFloat(Minimum=0.2, Maximum=0.4)
    Index 0 Param 2=1
    Module.Index0Param0Write=true
    Module.Index0Param1Write=true
    Module.Index0Param2Write=true
    Module.Index0Param3Write=true
   )
   &AddVelocity(Velocity=&RandomRangeVector(Minimum=<-100 -100 -100>, Maximum=<100 100 100>))
   &SampleParticlesFromOtherEmitter()
  ]
  particleUpdateScript [
   &ParticleState()
   &ScaleSpriteSize(Uniform Curve Sprite Scale=DataInterface() {propertyValues {Curve {keys [
        {tangentMode=RCTM_User, time=0, value=0.05}
        <at 0.2024 0.4>
        <at 0.291 1.45>
        {interpMode=RCIM_Cubic, tangentMode=RCTM_Break, time=0.3954, value=0.7}
        {arriveTangent=-2.0701, interpMode=RCIM_Cubic, tangentMode=RCTM_Break, time=1, value=0}
       ]}}})
   &GravityForce(Gravity=&Multiply_VectorByFloat(
     Vector=<1 1 -100>
     Float=&FloatFromCurve(FloatCurve=DataInterface() {propertyValues {Curve {keys [
          {interpMode=RCIM_Cubic, leaveTangent=1.2256, tangentMode=RCTM_Break, time=0, value=0.05}
          {arriveTangent=1.148, tangentMode=RCTM_User, time=1, value=0.85}
         ]}}})
    ))
   &CurlNoiseForce(Bias Noise=false, Pan Noise=false)
   &SolveForcesAndVelocity(Clamp Velocity=false, Limit Acceleration=false)
   &ScaleColor(
    Scale RGB=&VectorFromFloat(Value=&FloatFromCurve(FloatCurve=DataInterface() {propertyValues {Curve {keys [
          {interpMode=RCIM_Cubic, leaveTangent=-9.1901, tangentMode=RCTM_Break, time=0, value=1}
          {arriveTangent=-0.7291, interpMode=RCIM_Cubic, leaveTangent=-0.7291, tangentMode=RCTM_User, time=0.12, value=0.3}
          <at 1 0>
         ]}}}))
    ScaleA=true
    ScaleRGB=true
   )
   &FN_Light_Attributes(
    Color=$Particles.Color
    Create Color Attribute=true
    Create Exponent Attribute=true
    Create Radius Attribute=true
    Exponent=2
    Radius=45
    Scale Color=0.5
   )
   /SetVariables(Particles.Sparkle 2:Vector2f=&Vector2DFromFloat(Value=&FloatFromCurve(
      FloatCurve=DataInterface() {propertyValues {Curve {
         defaultValue=1
         keys [{interpMode=RCIM_Cubic, tangentMode=RCTM_User, time=0, value=1}, {interpMode=RCIM_Cubic, tangentMode=RCTM_User, time=1, value=0}]
        }}}
      Scale Curve=&RandomRangeFloat(Minimum=15, Maximum=25)
     ))) {isSetParameters=true}
  ]
  rendererClasses [&NiagaraSpriteRendererProperties]
  renderers [
   {rendererClass=&NiagaraSpriteRendererProperties}
   {rendererClass=&NiagaraSpriteRendererProperties, rendererIndex=1}
  ]
 }
 rope_shape {
  emitterUpdateScript [
   &EmitterState(Life Cycle Mode=::NewEnumerator1, Loop Behavior=::NewEnumerator1, Loop Duration=3, UseLoopDelay=false)
   &SpawnBurst_Instantaneous(Use Spawn Probability=false, UseLoopCountLimit=false)
   &SpawnRate(
    SpawnRate=&Multiply_Float(
     A=6
     B=&FloatFromCurve(
      FloatCurve=DataInterface() {propertyValues {Curve {keys [
          {arriveTangent=-7.1176, interpMode=RCIM_Cubic, leaveTangent=-7.1176, tangentMode=RCTM_User, time=0.2235, value=1}
          {arriveTangentWeight=0.2088, interpMode=RCIM_Cubic, tangentMode=RCTM_User, time=0.65, value=0}
         ]}}}
      CurveIndex=$Emitter.NormalizedLoopAge
     )
    )
    Use Spawn Probability=false
   ) {enabled=false, module=SpawnRate001}
  ]
  particleSpawnScript [
   &PartitionParticles()
   &InitializeParticle(
    Lifetime=0.2
    Sprite Size Mode=::NewEnumerator1
    Uniform Sprite Size Max=0
    Uniform Sprite Size Min=0
    UsePositionOffset=false
    Write Material Random=true
   ) {module=InitializeParticle001}
  ]
  particleUpdateScript [
   &ParticleState()
   &SystemLocation(
    MaskSpawnGroup=false
    Offset=&MaskVector(
     Mask Source=1
     Mask Parameter=$Particles.PartitionParticles.Partition
     Mask=0
     Masked Vector=&MakeVector(
      X=&Add_Float(
       A=12
       B=&FloatFromCurve(
        FloatCurve=DataInterface() {propertyValues {Curve {keys [
            {arriveTangent=1.0335, interpMode=RCIM_Cubic, leaveTangent=1.0335, tangentMode=RCTM_User, time=0, value=0}
            {arriveTangent=-5.7115, interpMode=RCIM_Cubic, leaveTangent=-5.7115, tangentMode=RCTM_User, time=0.1176, value=-0.5241}
            {arriveTangent=7.196, interpMode=RCIM_Cubic, leaveTangent=7.196, tangentMode=RCTM_User, time=0.1998, value=-0.5078}
            {arriveTangent=3.0037, interpMode=RCIM_Cubic, leaveTangent=3.0037, tangentMode=RCTM_User, time=0.3013, value=0.1474}
            {arriveTangent=-3.2136, interpMode=RCIM_Cubic, leaveTangent=-3.2136, tangentMode=RCTM_User, time=0.5783, value=-0.0795}
            {arriveTangent=-0.699, interpMode=RCIM_Cubic, leaveTangent=-0.699, time=0.7884, value=-0.6149}
            {arriveTangent=3.0036, interpMode=RCIM_Cubic, leaveTangent=3.0036, tangentMode=RCTM_User, time=1.0145, value=-0.3844}
           ]}}}
        Scale Curve=22
       )
      )
      Y=0
      Z=&Add_Float(
       A=1
       B=&FloatFromCurve(
        FloatCurve=DataInterface() {propertyValues {Curve {
           defaultValue=0
           keys [
            {interpMode=RCIM_Cubic, tangentMode=RCTM_User, time=0, value=0}
            {interpMode=RCIM_Cubic, tangentMode=RCTM_User, time=0.2427, value=0.5}
            {tangentMode=RCTM_User, time=0.5, value=0}
           ]
          }}}
        Scale Curve=-75
       )
      )
     )
     Passthrough Vector=<1 1 1>
    )
   )
  ]
  rendererClasses [&NiagaraSpriteRendererProperties]
  renderers [{rendererClass=&NiagaraSpriteRendererProperties}]
 }
}
summary {
 systemName=NS_Emote_IlluminateLasso_PropSpawn
 emitters [
  {emitterName=Flares, rendererClasses [&NiagaraSpriteRendererProperties]}
  {emitterName=rope_shape, rendererClasses [&NiagaraSpriteRendererProperties]}
 ]
 userVariables []
}
