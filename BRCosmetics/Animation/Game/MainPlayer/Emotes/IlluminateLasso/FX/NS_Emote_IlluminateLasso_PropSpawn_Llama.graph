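# Niagara effect for the llama prop spawn in the Illuminate Lasso emote: bursts of colored flares, sparkles and smoke.
# Flare and sparkle emitters pick from a three-color array tinted by a time-of-day color, with brightness also scaled by time of day.
# Particles sample a skeletal mesh; sparkles spread across four partitioned bone or socket indices and swirl with curl noise.
# Smoke size and velocity respond to ForwardDot, and its world/local blend depends on owner speed.
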NiagaraSystem /BRCosmetics/Animation/Game/MainPlayer/Emotes/IlluminateLasso/FX/NS_Emote_IlluminateLasso_PropSpawn_Llama
dependencies {
 usedDataInterfaces [&NiagaraDataInterfaceArrayColor, &NiagaraDataInterfaceCurve]
 usedDynamicInputs [&FNTimeOfDayColor]
 usedModules [
  &self:SystemSpawnScript.SystemScriptSource.SystemScriptGraph.NiagaraNodeAssignment_0.SetVariables_#1
  &SystemState
  &BlendSpaceHelper
  &self:SystemSpawnScript.SystemScriptSource.SystemScriptGraph.NiagaraNodeAssignment_1.SetVariables_#2
 ]
}
emitters {
 Burst_Flares {
  emitterUpdateScript [
   &EmitterState()
   &SpawnBurst_Instantaneous(Spawn Count=35, Spawn Time=0.2, Use Spawn Probability=false, UseLoopCountLimit=false)
   &SpawnRate(SpawnRate=8, Use Spawn Probability=false) {enabled=false}
  ]
  particleSpawnScript [
   &InitializeParticle(
    Color=&MultiplyLinearColor(
     Linear Color=&SelectLinearColorFromArray(
      Array Sampling Mode=::NewEnumerator0
      Direct Array Index=?
      Color Selection Array=DataInterface() {propertyValues {ColorData [<rgba 2.2045 2.4034 3 1>, <rgba 6 2.3575 0 1>, <rgba 3 0.1101 0 1>]}}
      Random Seed=0
     )
     Scale Factor=&MakeVector4_FromLinearColor(LinearColor=$System.TOD)
    )
    Lifetime Max=0.5
    Lifetime Min=0.35
    Lifetime Mode=::NewEnumerator1
    Sprite Rotation Mode=::NewEnumerator2
    Sprite Size Mode=::NewEnumerator1
    Uniform Sprite Size Max=15
    UsePositionOffset=false
    Write Material Random=true
   )
   &PartitionParticles(Number Of Partitions=1) {enabled=false}
   &SkeletalMeshLocation(
    Bone / Socket Index=$Particles.PartitionParticles.Partition
    Bone Sampling Mode=::NewEnumerator4
    MaskSpawnGroup=false
    Skeletal Mesh=$System.Skeletal Mesh
   )
   &ShapeLocation(MaskSpawnGroup=false, Override Seed=false, Sphere Radius=25, UseOwnerScale=true)
   &DynamicMaterialParameters(
    Index 0 Param 1=&RandomRangeFloat(Minimum=0.2, Maximum=0.4)
    Index 0 Param 2=1
    Module.Index0Param0Write=true
    Module.Index0Param1Write=true
    Module.Index0Param2Write=true
    Module.Index0Param3Write=true
   )
   &AddVelocity(Velocity=&RandomRangeVector(Minimum=<-100 -100 -100>, Maximum=<100 100 100>))
   /SetVariables() {isSetParameters=true, module=SetVariables_#3}
  ]
  particleUpdateScript [
   &ParticleState()
   &ScaleSpriteSize(Uniform Curve Sprite Scale=DataInterface() {propertyValues {Curve {keys [
        {tangentMode=RCTM_User, time=0, value=0.05}
        <at 0.2024 0.4>
        <at 0.291 1.45>
        {interpMode=RCIM_Cubic, tangentMode=RCTM_Break, time=0.3954, value=0.7}
        {arriveTangent=-2.0701, interpMode=RCIM_Cubic, tangentMode=RCTM_Break, time=1, value=0}
       ]}}})
   &GravityForce(Gravity=&Multiply_VectorByFloat(
     Vector=<1 1 -100>
     Float=&FloatFromCurve(FloatCurve=DataInterface() {propertyValues {Curve {keys [
          {interpMode=RCIM_Cubic, leaveTangent=1.2256, tangentMode=RCTM_Break, time=0, value=0.05}
          {arriveTangent=1.148, tangentMode=RCTM_User, time=1, value=0.85}
         ]}}})
    ))
   &Drag(
    Drag=&FloatFromCurve(FloatCurve=DataInterface() {propertyValues {Curve {keys [<at 0 1>,<at 0.5 0>]}}}, Scale Curve=5)
    Use Linear Drag=true
    Use Rotational Drag=true
   )
   &AddRotationalVelocity(Rotation Rate=<0.2 2 2>)
   &SolveForcesAndVelocity(Clamp Velocity=false, Limit Acceleration=false)
   &ScaleColor(
    Scale Alpha=&FNTimeOfDayScalar(Night Scale=2, Morning Scale=2, Day Scale=2, Evening Scale=2)
    Scale RGB=&VectorFromFloat(Value=&FloatFromCurve(FloatCurve=DataInterface() {propertyValues {Curve {
         defaultValue=1
         keys [{interpMode=RCIM_Cubic, tangentMode=RCTM_User, time=0, value=1}, {interpMode=RCIM_Cubic, tangentMode=RCTM_User, time=1, value=0}]
        }}}))
    ScaleA=true
    ScaleRGB=true
   )
  ]
  rendererClasses [&NiagaraSpriteRendererProperties]
  renderers [
   {rendererClass=&NiagaraSpriteRendererProperties}
   {rendererClass=&NiagaraSpriteRendererProperties, rendererIndex=1}
  ]
 }
 Flares001 {
  emitterUpdateScript [
   &EmitterState()
   &SpawnBurst_Instantaneous(Spawn Count=20, Spawn Time=0.23, Use Spawn Probability=false, UseLoopCountLimit=false)
   &SpawnRate(SpawnRate=8, Use Spawn Probability=false) {enabled=false}
  ]
  particleSpawnScript [
   &InitializeParticle(
    Color=&MultiplyLinearColor(
     Linear Color=&SelectLinearColorFromArray(
      Array Sampling Mode=::NewEnumerator0
      Direct Array Index=?
      Color Selection Array=DataInterface() {propertyValues {ColorData [<rgba 2.2045 2.4034 3 1>, <rgba 6 2.3575 0 1>, <rgba 3 0.1101 0 1>]}}
      Random Seed=0
     )
     Scale Factor=&MakeVector4_FromLinearColor(LinearColor=$System.TOD)
    )
    Lifetime Max=0.5
    Lifetime Min=0.35
    Lifetime Mode=::NewEnumerator1
    Sprite Rotation Mode=::NewEnumerator2
    Sprite Size Mode=::NewEnumerator1
    Uniform Sprite Size Max=15
    UsePositionOffset=false
    Write Material Random=true
   )
   &PartitionParticles(Number Of Partitions=1) {enabled=false}
   &SkeletalMeshLocation(
    Bone / Socket Index=$Particles.PartitionParticles.Partition
    Bone Sampling Mode=::NewEnumerator4
    MaskSpawnGroup=false
    Skeletal Mesh=$System.Skeletal Mesh
   )
   &ShapeLocation(MaskSpawnGroup=false, Override Seed=false, Sphere Radius=25, UseOwnerScale=true)
   &DynamicMaterialParameters(
    Index 0 Param 1=&RandomRangeFloat(Minimum=0.2, Maximum=0.4)
    Index 0 Param 2=1
    Module.Index0Param0Write=true
    Module.Index0Param1Write=true
    Module.Index0Param2Write=true
    Module.Index0Param3Write=true
   )
   &AddVelocity(
    Cone Angle=90
    Cone Axis=<-1 1 0>
    Module.UseConeFalloff=false
    Module.UseVelDistribution=false
    Override Seed=false
    Velocity Mode=::NewEnumerator2
    Velocity Speed=&RandomRangeFloat(Minimum=250, Maximum=600)
   )
  ]
  particleUpdateScript [
   &ParticleState()
   &ScaleSpriteSize(Uniform Curve Sprite Scale=DataInterface() {propertyValues {Curve {keys [
        {tangentMode=RCTM_User, time=0, value=0.05}
        <at 0.2024 0.4>
        <at 0.291 1.45>
        {interpMode=RCIM_Cubic, tangentMode=RCTM_Break, time=0.3954, value=0.7}
        {arriveTangent=-2.0701, interpMode=RCIM_Cubic, tangentMode=RCTM_Break, time=1, value=0}
       ]}}})
   &CurlNoiseForce(Noise Frequency=5, Noise Strength=300, Pan Noise=true, Pan Noise Field=<0 0.2 0.5>)
   &Drag(Drag=&UniformRangedFloat(Minimum=2, Maximum=15), Use Linear Drag=true, Use Rotational Drag=false)
   &GravityForce(Gravity=<0 0 0>)
   &SolveForcesAndVelocity(Clamp Velocity=false, Limit Acceleration=false)
   &ScaleColor(
    Scale Alpha=&FNTimeOfDayScalar(Night Scale=2, Morning Scale=2, Day Scale=2, Evening Scale=2)
    Scale RGB=&VectorFromFloat(Value=&FloatFromCurve(FloatCurve=DataInterface() {propertyValues {Curve {
         defaultValue=1
         keys [{interpMode=RCIM_Cubic, tangentMode=RCTM_User, time=0, value=1}, {interpMode=RCIM_Cubic, tangentMode=RCTM_User, time=1, value=0}]
        }}}))
    ScaleA=true
    ScaleRGB=true
   )
  ]
  rendererClasses [&NiagaraSpriteRendererProperties]
  renderers [
   {rendererClass=&NiagaraSpriteRendererProperties}
   {rendererClass=&NiagaraSpriteRendererProperties, rendererIndex=1}
  ]
 }
 Smoke {
  emitterUpdateScript [
   &EmitterState(Life Cycle Mode=::NewEnumerator1, Loop Behavior=::NewEnumerator1, Loop Duration=1.5, UseLoopDelay=false)
   &SpawnRate(SpawnRate=25, Use Spawn Probability=false)
  ]
  particleSpawnScript [
   &InitializeParticle(
    Color=<rgba 0.4492 0.6108 0.8177 0.2>
    Lifetime Max=0.7
    Lifetime Min=0.4
    Lifetime Mode=::NewEnumerator1
    Sprite Rotation Mode=::NewEnumerator2
    Sprite Size Mode=::NewEnumerator1
    Uniform Sprite Size Max=100
    Uniform Sprite Size Min=50
    UsePositionOffset=false
    Write Material Random=true
   )
   &TorusLocation(
    Align Torus With Axis=false
    Handle Radius=0
    Large Radius:NiagaraFloat=&Multiply_Float(A=30, B=&ClampFloat(Float=$User.ForwardDot, Min=0.3))
    MaskSpawnGroup=false
    Module.Override Randomness=false
   )
   &AddVelocityFromPoint(Module.Use Falloff Distance=false, Origin Offset=<0 150 0>, Velocity Strength=200)
   &AddVelocity(Velocity=<0 -100 0>)
   &AddVelocityInCone(Cone Angle=10, Cone Axis=<0 -1 0>, Module.Use Velocity Falloff On Cone Axis=false, Velocity Strength=300) {module=AddVelocityInCone001}
   &CameraOffset(Camera Offset Amount=10)
   &DynamicMaterialParameters(
    Index 0 Param 1=1
    Index 0 Param 2=1
    Index 0 Param 3=1
    Index 0 Param 4=1
    Module.Param0WriteEnabled=true
    Module.Param1WriteEnabled=true
    Module.Param2WriteEnabled=true
    Module.Param3WriteEnabled=true
   )
   &SkeletalMeshLocation(Bone Sampling Mode=::NewEnumerator4, MaskSpawnGroup=false, Position Sampling=::NewEnumerator1, Skeletal Mesh=$System.Skeletal Mesh)
  ]
  particleUpdateScript [
   &AccelerationForce(Acceleration=&MakeVector(X=0, Y=0, Z=300))
   &ScaleVelocity(Velocity Scale=&Multiply_VectorByFloat(
     Vector=<1 1 1>
     Float=&FN_FrontEndFloatModifier(Gameplay=&ClampFloat(Float=$User.ForwardDot, Min=0.5), Front End=1)
    ))
   &Drag(Drag=&RandomRangeFloat(Minimum=2, Maximum=5), Use Linear Drag=true, Use Rotational Drag=true)
   &SolveForcesAndVelocity(Clamp Velocity=false, Limit Acceleration=false)
   &ParticleState()
   &ScaleColor(
    Scale Alpha=&Multiply_Float(
     A=&FloatFromCurve(
      FloatCurve=DataInterface() {propertyValues {Curve {keys [
          {interpMode=RCIM_Cubic, tangentMode=RCTM_User, time=0, value=0}
          {interpMode=RCIM_Cubic, leaveTangent=-2.6773, tangentMode=RCTM_Break, time=0.3, value=1}
          {tangentMode=RCTM_User, time=0.85, value=0}
         ]}}}
      Scale Curve=&RandomRangeFloat(Minimum=0.5, Maximum=1)
     )
     B=&FNTimeOfDayScalar(Night Scale=0.1, Morning Scale=0.2, Day Scale=0.2, Evening Scale=0.3)
    )
    ScaleA=true
    ScaleRGB=true
   )
   &SubUVAnimation()
   &ScaleSpriteSize(Uniform Curve Sprite Scale=DataInterface() {propertyValues {Curve {keys [
        {arriveTangent=4.1785, interpMode=RCIM_Cubic, leaveTangent=4.1785, tangentMode=RCTM_User, time=0, value=0}
        {arriveTangent=0.9995, interpMode=RCIM_Cubic, leaveTangent=0.9995, tangentMode=RCTM_User, time=0.2223, value=0.6825}
        {arriveTangent=-0.0405, interpMode=RCIM_Cubic, leaveTangent=-0.0405, tangentMode=RCTM_User, time=1, value=1}
       ]}}})
   &BlendSpace("Position Blend (World <-> Local)":NiagaraFloat=&MakeCustomFloatFromBool(
     Bool=&SetBoolByFloatComparison(A=&VectorLength(Vector=$Engine.Owner.Velocity), B=400)
     True Float=1
     False Float=&FloatFromCurve(FloatCurve=DataInterface() {propertyValues {Curve {keys [<at 0 1 RCIM_Cubic>, {arriveTangent=-3.5186, interpMode=RCIM_Cubic, leaveTangent=-3.5186, tangentMode=RCTM_User, time=1, value=0}]}}})
     Translation Contribution:NiagaraFloat=1
     Rotation Contribution:NiagaraFloat=1
    ))
  ]
  rendererClasses [&NiagaraSpriteRendererProperties]
  renderers [{rendererClass=&NiagaraSpriteRendererProperties}]
 }
 Sparkles {
  emitterUpdateScript [
   &EmitterState(Life Cycle Mode=::NewEnumerator1, Loop Behavior=::NewEnumerator1, Loop Delay=0.1, Loop Duration=1.75, UseLoopDelay=true)
   &SpawnPerUnit(Module.Use Max Movement Threshold=true, Module.Use Movement Tolerance=false, Use Spawn Probability=false) {enabled=false}
   &SpawnRate(SpawnRate=100, Use Spawn Probability=false)
  ]
  particleSpawnScript [
   &InitializeParticle(
    Color=&MultiplyLinearColor(
     Linear Color=&SelectLinearColorFromArray(
      Array Sampling Mode=::NewEnumerator0
      Direct Array Index=?
      Color Selection Array=DataInterface() {propertyValues {ColorData [<rgba 2.2045 2.4034 3 1>, <rgba 6 2.3575 0 1>, <rgba 3 0.1101 0 1>]}}
      Random Seed=0
     )
     Scale Factor=&MakeVector4_FromLinearColor(LinearColor=$System.TOD)
    )
    Lifetime Max=0.5
    Lifetime Min=0.25
    Lifetime Mode=::NewEnumerator1
    Sprite Rotation Mode=::NewEnumerator2
    Sprite Size Mode=::NewEnumerator1
    Uniform Sprite Size Max=30
    UsePositionOffset=false
    Write Material Random=true
   )
   &PartitionParticles(Number Of Partitions=4)
   &ShapeLocation(MaskSpawnGroup=false, Override Seed=false, Sphere Radius=20, Sphere Surface Distribution=1, UseOwnerScale=true)
   &InheritVelocity(Speed Limit=true, UseSpeedThreshold=false)
   &SkeletalMeshLocation(
    Bone / Socket Index=$Particles.PartitionParticles.Partition
    Bone Sampling Mode=::NewEnumerator4
    MaskSpawnGroup=false
    Position Sampling=::NewEnumerator1
    Sampled Position Offset:NiagaraPosition=$Particles.ShapeLocation.ShapePosition
    Skeletal Mesh=$System.Skeletal Mesh
   )
   &DynamicMaterialParameters(
    Index 0 Param 1=&RandomRangeFloat(Minimum=0.3, Maximum=0)
    Index 0 Param 2=&RandomRangeFloat(Minimum=1, Maximum=2)
    Index 0 Param 3=&RandomRangeFloat(Minimum=-1, Maximum=1)
    Module.Index0Param0Write=true
    Module.Index0Param1Write=true
    Module.Index0Param2Write=true
    Module.Index0Param3Write=true
   )
   &AddVelocity(
    Velocity=&MakeVector(
     X=0
     Y=&FloatFromCurve(FloatCurve=DataInterface() {propertyValues {Curve {keys [<at 0 1>,<at 1 0>]}}}, Scale Curve=-100)
     Z=&FloatFromCurve(
      FloatCurve=DataInterface() {propertyValues {Curve {
         defaultValue=1
         keys [{interpMode=RCIM_Cubic, tangentMode=RCTM_User, time=0, value=1}, {interpMode=RCIM_Cubic, tangentMode=RCTM_User, time=0.35, value=0}]
        }}}
      Scale Curve=50
     )
    )
    Velocity Speed Scale=2
   )
  ]
  particleUpdateScript [
   &ParticleState()
   &ScaleColor(
    Scale Alpha=&FNTimeOfDayScalar(Night Scale=2, Morning Scale=1, Evening Scale=2)
    ScaleA=true
    ScaleRGB=true
   )
   &ScaleSpriteSize(Uniform Curve Sprite Scale=DataInterface() {propertyValues {Curve {
       defaultValue=0.05
       keys [
        {arriveTangent=9.0725, interpMode=RCIM_Cubic, leaveTangent=9.0725, tangentMode=RCTM_User, time=0, value=0}
        {arriveTangent=-2.5249, interpMode=RCIM_Cubic, leaveTangent=-2.5249, tangentMode=RCTM_User, time=0.1408, value=0.9847}
        {arriveTangent=-0.0405, interpMode=RCIM_Cubic, leaveTangent=-0.0405, tangentMode=RCTM_User, time=1, value=0}
       ]
      }}})
   &GravityForce(Gravity=<0 0 -50>)
   &Drag(Drag=2.5, Use Linear Drag=true, Use Rotational Drag=true)
   &CurlNoiseForce(Bias Noise=false, Noise Strength=100, Pan Noise=false) {enabled=false}
   &CurlNoiseForce(
    Noise Strength=&Multiply_Float(
     A=2000
     B=&FloatFromCurve(
      FloatCurve=DataInterface() {propertyValues {Curve {keys [
          {arriveTangent=0.0559, interpMode=RCIM_Cubic, leaveTangent=0.0559, tangentMode=RCTM_User, time=0, value=0.2}
          {arriveTangent=0.1895, interpMode=RCIM_Cubic, leaveTangent=0.1895, tangentMode=RCTM_User, time=0.9, value=0.25}
          {arriveTangent=1.1808, interpMode=RCIM_Cubic, leaveTangent=1.1808, tangentMode=RCTM_User, time=1, value=0.3}
         ]}}}
      CurveIndex=&RandomRangeFloat(Minimum=0, Maximum=1)
     )
    )
    Pan Noise=true
    Pan Noise Field=<0.5 0.5 0>
   ) {module=CurlNoiseForce001}
   &SolveForcesAndVelocity(Clamp Velocity=false, Limit Acceleration=false)
  ]
  rendererClasses [&NiagaraSpriteRendererProperties]
  renderers [
   {rendererClass=&NiagaraSpriteRendererProperties}
   {rendererClass=&NiagaraSpriteRendererProperties, rendererIndex=1}
  ]
 }
 center_flare {
  emitterUpdateScript [
   &EmitterState()
   &SpawnBurst_Instantaneous(Spawn Count=2, Spawn Time=0.2, Use Spawn Probability=false, UseLoopCountLimit=false)
   &SpawnRate(SpawnRate=8, Use Spawn Probability=false) {enabled=false}
  ]
  particleSpawnScript [
   &InitializeParticle(
    Color=&MultiplyLinearColor(
     Linear Color=&SelectLinearColorFromArray(
      Array Sampling Mode=::NewEnumerator0
      Direct Array Index=?
      Color Selection Array=DataInterface() {propertyValues {ColorData [<rgba 2.2045 2.4034 3 1>, <rgba 6 2.3575 0 1>, <rgba 3 0.1101 0 1>]}}
      Random Seed=0
     )
     Scale Factor=&MakeVector4_FromLinearColor(LinearColor=$System.TOD)
    )
    Lifetime Max=0.5
    Lifetime Min=0.35
    Lifetime Mode=::NewEnumerator1
    Sprite Rotation Mode=::NewEnumerator2
    Sprite Size Mode=::NewEnumerator4
    UsePositionOffset=false
    Write Material Random=true
   )
   &PartitionParticles(Number Of Partitions=1) {enabled=false}
   &SkeletalMeshLocation(
    Bone / Socket Index=$Particles.PartitionParticles.Partition
    Bone Sampling Mode=::NewEnumerator4
    MaskSpawnGroup=false
    Skeletal Mesh=$System.Skeletal Mesh
   )
   &ShapeLocation(MaskSpawnGroup=false, Override Seed=false, Sphere Radius=25, UseOwnerScale=true)
   &DynamicMaterialParameters(
    Index 0 Param 1=&RandomRangeFloat(Minimum=0.2, Maximum=0.4)
    Index 0 Param 2=1
    Module.Index0Param0Write=true
    Module.Index0Param1Write=true
    Module.Index0Param2Write=true
    Module.Index0Param3Write=true
   )
   /SetVariables() {isSetParameters=true, module=SetVariables_#3}
  ]
  particleUpdateScript [
   &ParticleState()
   &ScaleSpriteSize(
    Scale Sprite Size Mode=::NewEnumerator0
    Uniform Scale Factor=&FloatFromCurve(
     FloatCurve=DataInterface() {propertyValues {Curve {
        defaultValue=0
        keys [
         {interpMode=RCIM_Cubic, tangentMode=RCTM_Break, time=0, value=0}
         {interpMode=RCIM_Cubic, leaveTangent=-0.1735, tangentMode=RCTM_Break, time=0.15, value=3}
         {arriveTangent=-0.177, interpMode=RCIM_Cubic, leaveTangent=-0.177, tangentMode=RCTM_User, time=0.25, value=1}
         {arriveTangent=-0.15, interpMode=RCIM_Cubic, leaveTangent=-0.15, tangentMode=RCTM_User, time=0.9574, value=1}
        ]
       }}}
     Scale Curve=&MergeHelper_FloatSwitch(0=5, 1=1, 2=?, 3=?, Index=$Particles.VisibilityTag, Num=::NewEnumerator0)
    )
   )
   &ScaleColor(
    Scale Alpha=&FloatFromCurve(
     FloatCurve=DataInterface() {propertyValues {Curve {keys [<at 0.0018 0>,<at 0.2561 1>,<at 0.5 0>]}}}
     Scale Curve=&FNTimeOfDayScalar(Night Scale=0.25, Morning Scale=0.3, Day Scale=0.3, Evening Scale=0.3)
    )
    ScaleA=true
    ScaleRGB=true
   )
  ]
  rendererClasses [&NiagaraSpriteRendererProperties]
  renderers [
   {rendererClass=&NiagaraSpriteRendererProperties}
   {rendererClass=&NiagaraSpriteRendererProperties, rendererIndex=1}
   {rendererClass=&NiagaraSpriteRendererProperties, rendererIndex=2}
  ]
 }
}
summary {
 systemName=NS_Emote_IlluminateLasso_PropSpawn_Llama
 emitters [
  {emitterName=Sparkles, rendererClasses [&NiagaraSpriteRendererProperties]}
  {emitterName=Burst_Flares, rendererClasses [&NiagaraSpriteRendererProperties]}
  {emitterName=Flares001, rendererClasses [&NiagaraSpriteRendererProperties]}
  {emitterName=center_flare, rendererClasses [&NiagaraSpriteRendererProperties]}
  {emitterName=Smoke, rendererClasses [&NiagaraSpriteRendererProperties]}
 ]
 userVariables [
  User.bIsFrontEnd {type=NiagaraBool}
  User.bIsFrontEndPreview {type=NiagaraBool}
  User.ForwardDot {type=NiagaraFloat, defaultValue=0, description=""}
  User.OwnerLocalVelocity {type=Vector3f, description="", defaultValue {struct=Vector3f, value=<0 0 0>}}
  User.RightDot {type=NiagaraFloat, defaultValue=0, description=""}
  User.SkeletalMesh {
   type=NiagaraDataInterfaceSkeletalMesh
   defaultValue {
    struct=NiagaraExt_VariableValue_DataInterface
    value {dataInterface=&self:NiagaraDataInterfaceSkeletalMesh_0}
   }
  }
 ]
}
