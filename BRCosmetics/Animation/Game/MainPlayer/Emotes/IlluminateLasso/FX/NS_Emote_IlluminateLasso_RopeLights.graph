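# Niagara system for the Illuminate Lasso emote that puts blinking, multicolored sprite lights along the lasso prop's rope.
# The active Lights001 emitter bursts 12 particles, using partition index to pin each to a rope bone and pick a repeating green, amber, red or blue color.
# Waveforms offset by partition and system blink speed drive glow size, alpha and material parameter, with brightness scaled by time of day.
# An alternative 13-light emitter using curve-driven flicker is disabled.
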NiagaraSystem /BRCosmetics/Animation/Game/MainPlayer/Emotes/IlluminateLasso/FX/NS_Emote_IlluminateLasso_RopeLights
dependencies {
 usedDataInterfaces [&NiagaraDataInterfaceArrayColor, &NiagaraDataInterfaceCurve, &NiagaraDataInterfaceSkeletalMesh]
 usedDynamicInputs [&FNTimeOfDayColor]
 usedModules [
  &SystemState
  &self:SystemSpawnScript.SystemScriptSource.SystemScriptGraph.NiagaraNodeAssignment_1.SetVariables_#1
  &self:SystemSpawnScript.SystemScriptSource.SystemScriptGraph.NiagaraNodeAssignment_0.SetVariables_#2
 ]
}
emitters {
 Lights {
  bEnabled=false
  emitterUpdateScript [
   &EmitterState()
   &SpawnBurst_Instantaneous(Spawn Count=13, Use Spawn Probability=false, UseLoopCountLimit=false)
  ]
  particleSpawnScript [
   /SetVariables(Particles.VisibilityTag=&SpawnGroup()) {enabled=false, isSetParameters=true}
   &PartitionParticles(Number Of Partitions=13)
   &InitializeParticle(
    Color=&SelectLinearColorFromArray(
     Array Sampling Mode=::NewEnumerator1
     Direct Array Index=$Particles.PartitionParticles.Partition
     Color Selection Array=DataInterface() {propertyValues {ColorData [
        <rgba 0 3 0.0696 1>
        <rgba 5 2.3427 0.0244 1>
        <rgba 5 0.1342 0 1>
        <rgba 0 0.5614 5 1>
        <rgba 0 3 0.0696 1>
        <rgba 5 2.3427 0.0244 1>
        <rgba 5 0.1342 0 1>
        <rgba 0 0.5614 5 1>
        <rgba 0 3 0.0696 1>
        <rgba 5 2.3427 0.0244 1>
        <rgba 5 0.1342 0 1>
        <rgba 0 0.5614 5 1>
        <rgba 0 3 0.0696 1>
       ]}}
     Random Seed=?
    )
    Lifetime=5
    Sprite Size Mode=::NewEnumerator3
    Uniform Sprite Size=2.2
    UsePositionOffset=false
    Write Material Random=true
   )
   &CameraOffset(Camera Offset Amount=5)
  ]
  particleUpdateScript [
   &ParticleState()
   &SpriteRotationRate(Rotation Rate=&FloatFromCurve(FloatCurve=DataInterface() {propertyValues {Curve {keys [<at 0 1>,<at 0.5 0>]}}}, Scale Curve=360)) {enabled=false}
   &SkeletalMeshLocation(
    Bone / Socket Index=$Particles.PartitionParticles.Partition
    Bone Sampling Mode=::NewEnumerator4
    MaskSpawnGroup=false
    Orientation Sampling=::NewEnumerator1
    Skeletal Mesh=DataInterface() {propertyValues {
      DefaultMesh=null
      ExcludeBoneName=root
      OptionalComponentClass=null
      PreviewMesh=/BRCosmetics/Animation/Props/Gadgets_IlluminateLasso/Meshes/IlluminateLasso_Gadget
      SkinningMode=SkinOnTheFly
      SoftSourceActor=null
      SourceMode=Default
      UvSetIndex=0
      WholeMeshLOD=-1
      bExcludeBone=false
      bReadDeformedGeometry=true
      bRequireCurrentFrameData=true
      ComponentTags []
      FilteredBones [
       JNT_RopeLicht_00
       JNT_RopeLicht_01
       JNT_RopeLicht_02
       JNT_RopeLicht_03
       JNT_RopeLicht_04
       JNT_RopeLicht_05
       JNT_RopeLicht_06
       JNT_RopeLicht_07
       JNT_RopeLicht_08
       JNT_RopeLicht_09
       JNT_RopeLicht_10
       JNT_RopeLicht_11
       JNT_RopeLicht_12
      ]
      FilteredSockets []
      MeshUserParameter {parameter {name=None, typeDefHandle {registeredTypeIndex=1}, varData []}}
      SamplingRegions []
     }}
   )
   &DynamicMaterialParameters(
    Index 0 Param 1=1
    Module.Index0Float0Write=true
    Module.Index0Float1Write=true
    Module.Index0Float2Write=true
    Module.Index0Float3Write=true
   )
   &ScaleColor(
    Scale Alpha=&FloatFromCurve(
     FloatCurve=DataInterface() {propertyValues {Curve {
        defaultValue=0
        keys [
         {interpMode=RCIM_Cubic, tangentMode=RCTM_User, time=0, value=0}
         {interpMode=RCIM_Cubic, tangentMode=RCTM_User, time=0.1261, value=1}
         {arriveTangent=1.7382, interpMode=RCIM_Cubic, leaveTangent=1.7382, time=0.2741, value=0.5}
         {arriveTangent=-0.558, interpMode=RCIM_Cubic, leaveTangent=-0.558, time=0.4138, value=1.5}
         {arriveTangent=-1.8721, interpMode=RCIM_Cubic, leaveTangent=-1.8721, time=0.5124, value=0.367}
         {arriveTangent=1.2359, interpMode=RCIM_Cubic, leaveTangent=1.2359, time=0.6274, value=1.1}
         {arriveTangent=0.5365, interpMode=RCIM_Cubic, leaveTangent=0.5365, time=0.7288, value=0.6345}
         {arriveTangent=-2.3392, interpMode=RCIM_Cubic, leaveTangent=-2.3392, time=0.8781, value=1.2345}
         {tangentMode=RCTM_User, time=1, value=0}
        ]
       }}}
     Scale Curve=&FNTimeOfDayScalar(Night Scale=5, Morning Scale=1, Evening Scale=1)
    )
    ScaleA=true
    ScaleRGB=true
   )
   /SetVariables(
    Particles.glow color:LinearColor=&MakeLinearColorFromVectorAndFloat(
     "Vector (RGB)"=&MakeVectorFromLinearColorRGB(Color=&SelectLinearColorFromArray(
       Array Sampling Mode=::NewEnumerator1
       Direct Array Index=$Particles.PartitionParticles.Partition
       Color Selection Array=DataInterface() {propertyValues {ColorData [
          <rgba 0 5 0.1161 1>
          <rgba 5 2.3427 0.0244 1>
          <rgba 5 0.1342 0 1>
          <rgba 0 0.5614 5 1>
          <rgba 0 5 0.1161 1>
          <rgba 5 2.3427 0.0244 1>
          <rgba 5 0.1342 0 1>
          <rgba 0 0.5614 5 1>
          <rgba 0 5 0.1161 1>
          <rgba 5 2.3427 0.0244 1>
          <rgba 5 0.1342 0 1>
          <rgba 0 0.5614 5 1>
          <rgba 0 5 0.1161 1>
         ]}}
       Random Seed=?
      ))
     "Float (Alpha)"=&FloatFromCurve(
      FloatCurve=DataInterface() {propertyValues {Curve {
         defaultValue=0
         keys [
          {interpMode=RCIM_Cubic, tangentMode=RCTM_User, time=0, value=0}
          {interpMode=RCIM_Cubic, tangentMode=RCTM_User, time=0.1261, value=1}
          {arriveTangent=1.7382, interpMode=RCIM_Cubic, leaveTangent=1.7382, time=0.2741, value=0.5}
          {arriveTangent=-0.558, interpMode=RCIM_Cubic, leaveTangent=-0.558, time=0.4138, value=1.5}
          {arriveTangent=-1.8721, interpMode=RCIM_Cubic, leaveTangent=-1.8721, time=0.5124, value=0.367}
          {arriveTangent=1.2359, interpMode=RCIM_Cubic, leaveTangent=1.2359, time=0.6274, value=1.1}
          {arriveTangent=0.5365, interpMode=RCIM_Cubic, leaveTangent=0.5365, time=0.7288, value=0.6345}
          {arriveTangent=-2.3392, interpMode=RCIM_Cubic, leaveTangent=-2.3392, time=0.8781, value=1.2345}
          {tangentMode=RCTM_User, time=1, value=0}
         ]
        }}}
      Scale Curve=&FNTimeOfDayScalar(Night Scale=2, Morning Scale=1, Evening Scale=1)
     )
    )
    Particles.glow size:Vector2f=&Vector2DFromFloat(Value=&RandomRangeFloat(
      Minimum=&FloatFromCurve(
       FloatCurve=DataInterface() {propertyValues {Curve {
          defaultValue=1
          keys [
           {interpMode=RCIM_Cubic, tangentMode=RCTM_User, time=0, value=0}
           {interpMode=RCIM_Cubic, tangentMode=RCTM_User, time=0.165, value=0.975}
           {arriveTangent=-0.1227, interpMode=RCIM_Cubic, leaveTangent=-0.1227, time=0.3668, value=0.492}
           {arriveTangent=-0.0127, interpMode=RCIM_Cubic, leaveTangent=-0.0127, time=0.5577, value=0.9268}
           {arriveTangent=-0.0689, interpMode=RCIM_Cubic, leaveTangent=-0.0689, time=0.7342, value=0.4873}
           {arriveTangent=-1.8333, interpMode=RCIM_Cubic, leaveTangent=-1.8333, time=0.8675, value=0.9055}
           {tangentMode=RCTM_User, time=1, value=0}
          ]
         }}}
       Scale Curve=3
      )
      Maximum=&FloatFromCurve(
       FloatCurve=DataInterface() {propertyValues {Curve {
          defaultValue=1
          keys [
           {interpMode=RCIM_Cubic, tangentMode=RCTM_User, time=0, value=0}
           {interpMode=RCIM_Cubic, tangentMode=RCTM_User, time=0.5, value=1}
           {tangentMode=RCTM_User, time=1, value=0}
          ]
         }}}
       Scale Curve=6.5
      )
     ))
   ) {isSetParameters=true, module=SetVariables_#3}
  ]
  rendererClasses [&NiagaraSpriteRendererProperties]
  renderers [
   {rendererClass=&NiagaraSpriteRendererProperties}
   {rendererClass=&NiagaraSpriteRendererProperties, rendererIndex=1}
   {rendererClass=&NiagaraSpriteRendererProperties, rendererIndex=2}
  ]
 }
 Lights001 {
  emitterSpawnScript [/SetVariables(Emitter.random glow age:NiagaraFloat=&RandomRangeFloat(Minimum=0.5, Maximum=1)) {isSetParameters=true}]
  emitterUpdateScript [
   &EmitterState()
   &SpawnBurst_Instantaneous(Spawn Count=12, Use Spawn Probability=false, UseLoopCountLimit=false)
  ]
  particleSpawnScript [
   &PartitionParticles(Number Of Partitions=12)
   &InitializeParticle(
    Color=&SelectLinearColorFromArray(
     Array Sampling Mode=::NewEnumerator1
     Direct Array Index=$Particles.PartitionParticles.Partition
     Color Selection Array=DataInterface() {propertyValues {ColorData [
        <rgba 0 3 0.0696 1>
        <rgba 5 2.3427 0.0244 1>
        <rgba 5 0.1342 0 1>
        <rgba 0 0.5614 5 1>
        <rgba 0 3 0.0696 1>
        <rgba 5 2.3427 0.0244 1>
        <rgba 5 0.1342 0 1>
        <rgba 0 0.5614 5 1>
        <rgba 0 3 0.0696 1>
        <rgba 5 2.3427 0.0244 1>
        <rgba 5 0.1342 0 1>
        <rgba 0 0.5614 5 1>
        <rgba 0 3 0.0696 1>
       ]}}
     Random Seed=?
    )
    Lifetime=4
    Sprite Size Mode=::NewEnumerator3
    Uniform Sprite Size=2.2
    UsePositionOffset=false
    Write Material Random=true
   )
   &CameraOffset(Camera Offset Amount=5)
  ]
  particleUpdateScript [
   &ParticleState()
   &SkeletalMeshLocation(
    Bone / Socket Index=$Particles.PartitionParticles.Partition
    Bone Sampling Mode=::NewEnumerator4
    MaskSpawnGroup=false
    Orientation Sampling=::NewEnumerator1
    Skeletal Mesh=DataInterface() {propertyValues {
      DefaultMesh=null
      ExcludeBoneName=root
      OptionalComponentClass=null
      PreviewMesh=&IlluminateLasso_Gadget
      SkinningMode=SkinOnTheFly
      SoftSourceActor=null
      SourceMode=Default
      UvSetIndex=0
      WholeMeshLOD=-1
      bExcludeBone=false
      bReadDeformedGeometry=true
      bRequireCurrentFrameData=true
      ComponentTags []
      FilteredBones [
       JNT_RopeLicht_01
       JNT_RopeLicht_02
       JNT_RopeLicht_03
       JNT_RopeLicht_04
       JNT_RopeLicht_05
       JNT_RopeLicht_06
       JNT_RopeLicht_07
       JNT_RopeLicht_08
       JNT_RopeLicht_09
       JNT_RopeLicht_10
       JNT_RopeLicht_11
       JNT_RopeLicht_12
      ]
      FilteredSockets []
      MeshUserParameter {parameter {name=None, typeDefHandle {registeredTypeIndex=1}, varData []}}
      SamplingRegions []
     }}
   )
   &ScaleColor(
    Scale Alpha=&FloatFromCurve(
     FloatCurve=DataInterface() {propertyValues {Curve {
        defaultValue=0
        keys [
         {interpMode=RCIM_Cubic, tangentMode=RCTM_User, time=0, value=0}
         {interpMode=RCIM_Cubic, tangentMode=RCTM_User, time=0.1261, value=1}
         {arriveTangent=1.7382, interpMode=RCIM_Cubic, leaveTangent=1.7382, time=0.2741, value=0.5}
         {arriveTangent=-0.558, interpMode=RCIM_Cubic, leaveTangent=-0.558, time=0.4138, value=1.5}
         {arriveTangent=-1.8721, interpMode=RCIM_Cubic, leaveTangent=-1.8721, time=0.5124, value=0.367}
         {arriveTangent=1.2359, interpMode=RCIM_Cubic, leaveTangent=1.2359, time=0.6274, value=1.1}
         {arriveTangent=0.5365, interpMode=RCIM_Cubic, leaveTangent=0.5365, time=0.7288, value=0.6345}
         {arriveTangent=-2.3392, interpMode=RCIM_Cubic, leaveTangent=-2.3392, time=0.8781, value=1.2345}
         {tangentMode=RCTM_User, time=1, value=0}
        ]
       }}}
     Scale Curve=&FNTimeOfDayScalar(Night Scale=5, Morning Scale=1, Evening Scale=1)
    )
    ScaleA=true
    ScaleRGB=true
   )
   /SetVariables(Particles.glow size:Vector2f=&Vector2DFromFloat(Value=&Waveform(
      Global Input Position=$Particles.Initial.PartitionParticles.NormalizedPartition
      Global Amplitude Scale=18
      "Waveform [1]"=::NewEnumerator3
      "[1] Phase Scale"=$System.blink Speed
      "[1] Frequency"=1
      "[1] Amplitude Min/Max"=<0.45 1.2>
     ))) {isSetParameters=true, module=SetVariables_#3}
   /SetVariables(Particles.glow 2:LinearColor=&MakeLinearColorFromVectorAndFloat(
     "Vector (RGB)"=&MakeVectorFromLinearColorRGB(Color=&SelectLinearColorFromArray(
       Array Sampling Mode=::NewEnumerator1
       Direct Array Index=$Particles.PartitionParticles.Partition
       Color Selection Array=DataInterface() {propertyValues {ColorData [
          <rgba 0 1 0.0232 1>
          <rgba 1 0.4685 0.0049 1>
          <rgba 1 0.0268 0 1>
          <rgba 0 0.1123 1 1>
          <rgba 0 1 0.0232 1>
          <rgba 1 0.4685 0.0049 1>
          <rgba 1 0.0268 0 1>
          <rgba 0 0.1123 1 1>
          <rgba 0 1 0.0232 1>
          <rgba 1 0.4685 0.0049 1>
          <rgba 1 0.0268 0 1>
          <rgba 0 0.1123 1 1>
          <rgba 0 1 0.0232 1>
         ]}}
       Random Seed=?
      ))
     "Float (Alpha)"=&Waveform(
      Global Input Position=$Particles.Initial.PartitionParticles.NormalizedPartition
      Global Amplitude Scale=$System.Rope Glows Alpha TOD
      Global Clamp Min=0.5
      "Waveform [1]"=::NewEnumerator0
      "[1] Phase Scale"=$System.blink Speed
      "[1] Frequency"=&RandomRangeFloat(Minimum=0.2, Maximum=1)
      "[1] Amplitude Min/Max"=<0.2 0.5>
     )
    )) {isSetParameters=true}
   &DynamicMaterialParameters(
    Index 0 Param 1=&Waveform(
     Global Input Position=$Particles.Initial.PartitionParticles.NormalizedPartition
     Global Amplitude Scale=$System.Rope Glows Alpha TOD
     "Waveform [1]"=::NewEnumerator3
     "[1] Phase Scale"=$System.blink Speed
     "[1] Frequency"=&RandomRangeFloat(Minimum=0.2, Maximum=1)
     "[1] Amplitude Min/Max"=<0.5 1>
    )
    Module.Index0Float0Write=true
    Module.Index0Float1Write=true
    Module.Index0Float2Write=true
    Module.Index0Float3Write=true
   )
  ]
  rendererClasses [&NiagaraSpriteRendererProperties]
  renderers [
   {rendererClass=&NiagaraSpriteRendererProperties}
   {rendererClass=&NiagaraSpriteRendererProperties, rendererIndex=1}
  ]
 }
}
summary {
 systemName=NS_Emote_IlluminateLasso_RopeLights
 emitters [
  {bEnabled=false, emitterName=Lights, rendererClasses [&NiagaraSpriteRendererProperties]}
  {emitterName=Lights001, rendererClasses [&NiagaraSpriteRendererProperties]}
 ]
 userVariables [
  User.bIsFrontEnd {type=NiagaraBool}
  User.bIsFrontEndPreview {type=NiagaraBool}
 ]
}
