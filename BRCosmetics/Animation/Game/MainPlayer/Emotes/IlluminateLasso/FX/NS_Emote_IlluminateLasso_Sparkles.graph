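# Niagara system for the Illuminate Lasso emote that sprays looping pulses of glowing sprite sparkles from bones or sockets on a skeletal mesh.
# Only Sparkles002 is enabled, merging the two disabled emitters through spawn groups that switch lifetime, size, velocity and gravity.
# Particles partition across four bone/socket indices, inherit velocity, then fall with drag while shrinking and fading.
# Time of day scales randomly picked colors and alpha, feeding five sprite renderers.
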NiagaraSystem /BRCosmetics/Animation/Game/MainPlayer/Emotes/IlluminateLasso/FX/NS_Emote_IlluminateLasso_Sparkles
dependencies {
 usedDataInterfaces [
  &NiagaraDataInterfaceArrayColor
  &NiagaraDataInterfaceCurve
  &NiagaraDataInterfaceVectorCurve
  &NiagaraDataInterfaceArrayFloat3
  &NiagaraDataInterfaceSkeletalMesh
 ]
 usedDynamicInputs [&FNTimeOfDayColor]
 usedModules [
  &self:SystemSpawnScript.SystemScriptSource.SystemScriptGraph.NiagaraNodeAssignment_0.SetVariables_#1
  &SystemState
  &self:SystemSpawnScript.SystemScriptSource.SystemScriptGraph.NiagaraNodeAssignment_1.SetVariables_#2
  &BlendSpaceHelper
 ]
}
emitters {
 Sparkles {
  bEnabled=false
  emitterUpdateScript [
   &EmitterState(Life Cycle Mode=::NewEnumerator1, Loop Delay=0.15, Loop Duration=0.35, Recalculate Duration Each Loop=true, UseLoopDelay=true)
   &SpawnPerUnit(Module.Use Max Movement Threshold=true, Module.Use Movement Tolerance=false, Use Spawn Probability=false) {enabled=false}
   &SpawnRate(SpawnRate=50, Use Spawn Probability=false)
  ]
  particleSpawnScript [
   &InitializeParticle(
    Color=&MultiplyLinearColor(
     Linear Color=&SelectLinearColorFromArray(
      Array Sampling Mode=::NewEnumerator0
      Direct Array Index=?
      Color Selection Array=DataInterface() {propertyValues {ColorData [<rgba 2.2045 2.4034 3 1>, <rgba 6 2.3575 0 1>, <rgba 3 0.1101 0 1>]}}
      Random Seed=0
     )
     Scale Factor=&MakeVector4_FromLinearColor(LinearColor=$System.TOD)
    )
    Lifetime Max=0.6
    Lifetime Min=0.5
    Lifetime Mode=::NewEnumerator1
    Sprite Rotation Angle=&RandomRangeFloat(Minimum=0, Maximum=360)
    Sprite Rotation Mode=::NewEnumerator0
    Sprite Size Mode=::NewEnumerator1
    Uniform Sprite Size Max=35
    Uniform Sprite Size Min=10
    UsePositionOffset=false
    Write Material Random=true
   )
   &PartitionParticles(Number Of Partitions=4)
   &ShapeLocation(MaskSpawnGroup=false, Override Seed=false, Sphere Radius=5, Sphere Surface Distribution=1, UseOwnerScale=true)
   &InheritVelocity(Speed Limit=true, UseSpeedThreshold=false)
   &SkeletalMeshLocation(
    Bone / Socket Index=$Particles.PartitionParticles.Partition
    Bone Sampling Mode=::NewEnumerator4
    MaskSpawnGroup=false
    Position Sampling=::NewEnumerator1
    Sampled Position Offset:NiagaraPosition=$Particles.ShapeLocation.ShapePosition
    Skeletal Mesh=$System.Skeletal Mesh
   )
   &DynamicMaterialParameters(
    Index 0 Param 1=&RandomRangeFloat(Minimum=0.3, Maximum=0)
    Index 0 Param 2=&RandomRangeFloat(Minimum=0.8, Maximum=1.5)
    Index 0 Param 3=&RandomRangeFloat(Minimum=-1, Maximum=1)
    Module.Index0Param0Write=true
    Module.Index0Param1Write=true
    Module.Index0Param2Write=true
    Module.Index0Param3Write=true
   )
   &AddVelocity(Velocity=&MakeVector(
     X=0
     Y=&FloatFromCurve(FloatCurve=DataInterface() {propertyValues {Curve {keys [<at 0 1>,<at 1 0>]}}}, Scale Curve=-80)
     Z=&FloatFromCurve(FloatCurve=DataInterface() {propertyValues {Curve {keys [<at 0 1>,<at 0.5 0>]}}}, Scale Curve=120)
    ))
  ]
  particleUpdateScript [
   &ParticleState()
   &SpriteRotationRate(Rotation Rate=&FloatFromCurve(FloatCurve=DataInterface() {propertyValues {Curve {
        defaultValue=1
        keys [{interpMode=RCIM_Cubic, tangentMode=RCTM_User, time=0, value=0}, {interpMode=RCIM_Cubic, tangentMode=RCTM_User, time=0.3, value=1}]
       }}})) {enabled=false}
   &ScaleSpriteSize(Uniform Curve Sprite Scale=DataInterface() {propertyValues {Curve {keys [
        {interpMode=RCIM_Cubic, tangentMode=RCTM_Break, time=0, value=0.05}
        {interpMode=RCIM_Cubic, tangentMode=RCTM_Break, time=0.2, value=0.85}
        {interpMode=RCIM_Cubic, tangentMode=RCTM_Break, time=0.5, value=0.95}
        {interpMode=RCIM_Cubic, tangentMode=RCTM_Break, time=1, value=0}
       ]}}})
   &GravityForce(Gravity=<0 0 -350>)
   &Drag(Drag=&RandomRangeFloat(Minimum=2.5, Maximum=5), Use Linear Drag=true, Use Rotational Drag=true)
   &AddRotationalVelocity(
    Coordinate Space:ENiagaraMeshTransforms=::NewEnumerator2
    Rotation Rate=&VectorFromCurve(VectorCurve=DataInterface() {propertyValues {XCurve {keys [<at 0 1>,<at 1 0>]}, YCurve {keys [<at 0 1>,<at 1 0>]}, ZCurve {keys [<at 0 1>,<at 1 0>]}}})
   )
   &SolveForcesAndVelocity(Clamp Velocity=false, Limit Acceleration=false)
   /SetVariables(Particles.sparkle sprite size:Vector2f=&RandomRangeVector2D(Minimum=<5 5>, Maximum=<10 10>)) {isSetParameters=true, module=SetVariables_#3}
   &ScaleColor(
    Scale Alpha=&FloatFromCurve(
     FloatCurve=DataInterface() {propertyValues {Curve {
        defaultValue=1
        keys [{interpMode=RCIM_Cubic, tangentMode=RCTM_User, time=0, value=1}, {interpMode=RCIM_Cubic, tangentMode=RCTM_User, time=1, value=0}]
       }}}
     Scale Curve=&FNTimeOfDayScalar(Night Scale=3, Morning Scale=2, Day Scale=2, Evening Scale=2)
    )
    ScaleA=true
    ScaleRGB=true
   )
   &BlendSpace(Rotation Contribution=0)
   /SetVariables(
    Particles.glow color:LinearColor=&MakeLinearColorFromVectorAndFloat(
     "Vector (RGB)"=&MakeVectorFromLinearColorRGB(Color=$Particles.Color)
     "Float (Alpha)"=&FloatFromCurve(
      FloatCurve=DataInterface() {propertyValues {Curve {
         defaultValue=1
         keys [{interpMode=RCIM_Cubic, tangentMode=RCTM_User, time=0, value=1}, {interpMode=RCIM_Cubic, tangentMode=RCTM_User, time=1, value=0}]
        }}}
      Scale Curve=&FNTimeOfDayScalar(Night Scale=2, Morning Scale=1, Evening Scale=2)
     )
    )
    Particles.sparkles color:LinearColor=&MakeLinearColorFromVectorAndFloat(
     "Vector (RGB)"=&MakeVectorFromLinearColorRGB(Color=$Particles.Color)
     "Float (Alpha)"=&FloatFromCurve(
      FloatCurve=DataInterface() {propertyValues {Curve {
         defaultValue=1
         keys [{interpMode=RCIM_Cubic, tangentMode=RCTM_User, time=0, value=1}, {interpMode=RCIM_Cubic, tangentMode=RCTM_User, time=1, value=0}]
        }}}
      Scale Curve=&FNTimeOfDayScalar(Night Scale=3, Morning Scale=2, Day Scale=2, Evening Scale=2)
     )
    )
   ) {isSetParameters=true, module=SetVariables_#4}
  ]
  rendererClasses [&NiagaraSpriteRendererProperties]
  renderers [
   {rendererClass=&NiagaraSpriteRendererProperties}
   {rendererClass=&NiagaraSpriteRendererProperties, rendererIndex=1}
   {rendererClass=&NiagaraSpriteRendererProperties, rendererIndex=2}
  ]
 }
 Sparkles001 {
  bEnabled=false
  emitterUpdateScript [
   &EmitterState(Life Cycle Mode=::NewEnumerator1, Loop Delay=0.15, Loop Duration=0.35, Recalculate Duration Each Loop=true, UseLoopDelay=true)
   &SpawnPerUnit(Module.Use Max Movement Threshold=true, Module.Use Movement Tolerance=false, Use Spawn Probability=false) {enabled=false}
   &SpawnRate(SpawnRate=35, Use Spawn Probability=false)
  ]
  particleSpawnScript [
   &InitializeParticle(
    Color=&MultiplyLinearColor(
     Linear Color=&SelectLinearColorFromArray(
      Array Sampling Mode=::NewEnumerator0
      Direct Array Index=?
      Color Selection Array=DataInterface() {propertyValues {ColorData [<rgba 2.2045 2.4034 3 1>, <rgba 6 2.3575 0 1>, <rgba 3 0.1101 0 1>]}}
      Random Seed=0
     )
     Scale Factor=&MakeVector4_FromLinearColor(LinearColor=$System.TOD)
    )
    Lifetime Max=0.5
    Lifetime Min=0.25
    Lifetime Mode=::NewEnumerator1
    Sprite Rotation Angle=&RandomRangeFloat(Minimum=0, Maximum=360)
    Sprite Rotation Mode=::NewEnumerator0
    Sprite Size Mode=::NewEnumerator1
    Uniform Sprite Size Max=25
    UsePositionOffset=false
    Write Material Random=true
   )
   &PartitionParticles(Number Of Partitions=4)
   &ShapeLocation(MaskSpawnGroup=false, Override Seed=false, Sphere Radius=10, Sphere Surface Distribution=1, UseOwnerScale=true)
   &InheritVelocity(Speed Limit=true, UseSpeedThreshold=false)
   &SkeletalMeshLocation(
    Bone / Socket Index=$Particles.PartitionParticles.Partition
    Bone Sampling Mode=::NewEnumerator4
    MaskSpawnGroup=false
    Position Sampling=::NewEnumerator1
    Sampled Position Offset:NiagaraPosition=$Particles.ShapeLocation.ShapePosition
    Skeletal Mesh=$System.Skeletal Mesh
   )
   &DynamicMaterialParameters(
    Index 0 Param 1=&RandomRangeFloat(Minimum=0.3, Maximum=0)
    Index 0 Param 2=&RandomRangeFloat(Minimum=0.8, Maximum=1.5)
    Index 0 Param 3=&RandomRangeFloat(Minimum=-1, Maximum=1)
    Module.Index0Param0Write=true
    Module.Index0Param1Write=true
    Module.Index0Param2Write=true
    Module.Index0Param3Write=true
   )
   &AddVelocity(Velocity=&MergeHelper_VectorSwitch(0=<-70 0 100>, 1=<70 0 100>, 2=<-70 0 100>, 3=<70 0 100>, Index=$Particles.PartitionParticles.Partition, Num=::NewEnumerator2))
  ]
  particleUpdateScript [
   &ParticleState()
   &ScaleSpriteSize(Uniform Curve Sprite Scale=DataInterface() {propertyValues {Curve {keys [
        {interpMode=RCIM_Cubic, tangentMode=RCTM_Break, time=0, value=0.05}
        {interpMode=RCIM_Cubic, tangentMode=RCTM_Break, time=0.2, value=0.85}
        {interpMode=RCIM_Cubic, tangentMode=RCTM_Break, time=0.5, value=0.95}
        {interpMode=RCIM_Cubic, tangentMode=RCTM_Break, time=1, value=0}
       ]}}})
   &GravityForce(Gravity=&MakeVector(X=0, Y=0, Z=&RandomRangeFloat(Minimum=-350, Maximum=-500)))
   &Drag(Drag=&RandomRangeFloat(Minimum=2.5, Maximum=5), Use Linear Drag=true, Use Rotational Drag=true)
   &AddRotationalVelocity(Rotation Rate=<0 1 0>)
   &SolveForcesAndVelocity(Clamp Velocity=false, Limit Acceleration=false)
   /SetVariables(Particles.sparkle sprite size:Vector2f=&RandomRangeVector2D(Minimum=<5 5>, Maximum=<10 10>)) {isSetParameters=true, module=SetVariables_#3}
   &ScaleColor(
    Scale Alpha=&FloatFromCurve(
     FloatCurve=DataInterface() {propertyValues {Curve {
        defaultValue=1
        keys [{interpMode=RCIM_Cubic, tangentMode=RCTM_User, time=0, value=1}, {interpMode=RCIM_Cubic, tangentMode=RCTM_User, time=1, value=0}]
       }}}
     Scale Curve=&FNTimeOfDayScalar(Night Scale=3, Morning Scale=2, Day Scale=2, Evening Scale=2)
    )
    ScaleA=true
    ScaleRGB=true
   )
   &BlendSpace(Rotation Contribution=0)
  ]
  rendererClasses [&NiagaraSpriteRendererProperties]
  renderers [
   {rendererClass=&NiagaraSpriteRendererProperties}
   {rendererClass=&NiagaraSpriteRendererProperties, rendererIndex=1}
   {rendererClass=&NiagaraSpriteRendererProperties, rendererIndex=2}
  ]
 }
 Sparkles002 {
  emitterUpdateScript [
   &EmitterState(Life Cycle Mode=::NewEnumerator1, Loop Delay=0.15, Loop Duration=0.35, Recalculate Duration Each Loop=true, UseLoopDelay=true)
   &SpawnRate(SpawnGroup=1, SpawnRate=35, Use Spawn Probability=false) {module=SpawnRate001}
   &SpawnRate(SpawnRate=50, Use Spawn Probability=false)
  ]
  particleSpawnScript [
   /SetVariables(Particles.VisibilityTag=&SpawnGroup()) {isSetParameters=true}
   &InitializeParticle(
    Color=&MultiplyLinearColor(
     Linear Color=&SelectLinearColorFromArray(
      Array Sampling Mode=::NewEnumerator0
      Direct Array Index=?
      Color Selection Array=DataInterface() {propertyValues {ColorData [<rgba 2.2045 2.4034 3 1>, <rgba 6 2.3575 0 1>, <rgba 3 0.1101 0 1>]}}
      Random Seed=0
     )
     Scale Factor=&MakeVector4_FromLinearColor(LinearColor=$System.TOD)
    )
    Lifetime=&MergeHelper_FloatSwitch(
     0=&RandomRangeFloat(Minimum=0.5, Maximum=0.6)
     1=&RandomRangeFloat(Minimum=0.25, Maximum=0.5)
     2=?
     3=?
     Index=&SpawnGroup()
     Num=::NewEnumerator0
    )
    Sprite Rotation Angle=&RandomRangeFloat(Minimum=0, Maximum=360)
    Sprite Rotation Mode=::NewEnumerator0
    Sprite Size Mode=::NewEnumerator3
    Uniform Sprite Size=&MergeHelper_FloatSwitch(
     0=&RandomRangeFloat(Minimum=10, Maximum=35)
     1=&RandomRangeFloat(Minimum=5, Maximum=25)
     2=?
     3=?
     Index=&SpawnGroup()
     Num=::NewEnumerator0
    )
    UsePositionOffset=false
    Write Material Random=true
   )
   &PartitionParticles(Number Of Partitions=4)
   &ShapeLocation(
    MaskSpawnGroup=false
    Override Seed=false
    Sphere Radius=&MergeHelper_FloatSwitch(0=5, 1=5, 2=?, 3=?, Index=&SpawnGroup(), Num=::NewEnumerator0)
    Sphere Surface Distribution=1
    UseOwnerScale=true
   )
   &InheritVelocity(Speed Limit=true, UseSpeedThreshold=false)
   &SkeletalMeshLocation(
    Bone / Socket Index=$Particles.PartitionParticles.Partition
    Bone Sampling Mode=::NewEnumerator4
    MaskSpawnGroup=false
    Position Sampling=::NewEnumerator1
    Sampled Position Offset:NiagaraPosition=$Particles.ShapeLocation.ShapePosition
    Skeletal Mesh=$System.Skeletal Mesh
   )
   &DynamicMaterialParameters(
    Index 0 Param 1=&RandomRangeFloat(Minimum=0.3, Maximum=0)
    Index 0 Param 2=&RandomRangeFloat(Minimum=0.8, Maximum=1.5)
    Index 0 Param 3=&RandomRangeFloat(Minimum=-1, Maximum=1)
    Module.Index0Param0Write=true
    Module.Index0Param1Write=true
    Module.Index0Param2Write=true
    Module.Index0Param3Write=true
   )
   &AddVelocity(Velocity=&MergeHelper_VectorSwitch(
     0=&MakeVector(
      X=0
      Y=&FloatFromCurve(FloatCurve=DataInterface() {propertyValues {Curve {keys [<at 0 1>,<at 1 0>]}}}, Scale Curve=-80)
      Z=&FloatFromCurve(FloatCurve=DataInterface() {propertyValues {Curve {keys [<at 0 1>,<at 0.5 0>]}}}, Scale Curve=120)
     )
     1=&SelectVectorFromArray(
      Array Sampling Mode=::NewEnumerator0
      Direct Array Index=?
      Vector Selection Array=DataInterface() {propertyValues {FloatData [<-70 0 100>, <70 0 100>, <70 0 100>, <-70 0 100>]}}
      Random Seed=0
     )
     2=?
     3=?
     Index=&SpawnGroup()
     Num=::NewEnumerator0
    ))
  ]
  particleUpdateScript [
   &ParticleState()
   &SpriteRotationRate(Rotation Rate=&FloatFromCurve(FloatCurve=DataInterface() {propertyValues {Curve {
        defaultValue=1
        keys [{interpMode=RCIM_Cubic, tangentMode=RCTM_User, time=0, value=0}, {interpMode=RCIM_Cubic, tangentMode=RCTM_User, time=0.3, value=1}]
       }}})) {enabled=false}
   &ScaleSpriteSize(Uniform Curve Sprite Scale=DataInterface() {propertyValues {Curve {keys [
        {interpMode=RCIM_Cubic, tangentMode=RCTM_Break, time=0, value=0.05}
        {interpMode=RCIM_Cubic, tangentMode=RCTM_Break, time=0.2, value=0.85}
        {interpMode=RCIM_Cubic, tangentMode=RCTM_Break, time=0.5, value=0.95}
        {interpMode=RCIM_Cubic, tangentMode=RCTM_Break, time=1, value=0}
       ]}}})
   &GravityForce(Gravity=&MergeHelper_VectorSwitch(
     0=<0 0 -350>
     1=&MakeVector(X=0, Y=0, Z=&RandomRangeFloat(Minimum=-350, Maximum=-500))
     2=?
     3=?
     Index=&SpawnGroup()
     Num=::NewEnumerator0
    ))
   &Drag(Drag=&RandomRangeFloat(Minimum=2.5, Maximum=5), Use Linear Drag=true, Use Rotational Drag=true)
   &AddRotationalVelocity(
    Coordinate Space:ENiagaraMeshTransforms=::NewEnumerator2
    Rotation Rate=&VectorFromCurve(VectorCurve=DataInterface() {propertyValues {XCurve {keys [<at 0 1>,<at 1 0>]}, YCurve {keys [<at 0 1>,<at 1 0>]}, ZCurve {keys [<at 0 1>,<at 1 0>]}}})
   )
   &SolveForcesAndVelocity(Clamp Velocity=false, Limit Acceleration=false)
   /SetVariables(Particles.sparkle sprite size:Vector2f=&RandomRangeVector2D(Minimum=<5 5>, Maximum=<10 10>)) {isSetParameters=true, module=SetVariables_#3}
   &ScaleColor(
    Scale Alpha=&FloatFromCurve(
     FloatCurve=DataInterface() {propertyValues {Curve {
        defaultValue=1
        keys [{interpMode=RCIM_Cubic, tangentMode=RCTM_User, time=0, value=1}, {interpMode=RCIM_Cubic, tangentMode=RCTM_User, time=1, value=0}]
       }}}
     Scale Curve=&FNTimeOfDayScalar(Night Scale=3, Morning Scale=2, Day Scale=2, Evening Scale=2)
    )
    ScaleA=true
    ScaleRGB=true
   )
   &BlendSpace(Rotation Contribution=0)
   /SetVariables(
    Particles.glow color:LinearColor=&MakeLinearColorFromVectorAndFloat(
     "Vector (RGB)"=&MakeVectorFromLinearColorRGB(Color=$Particles.Color)
     "Float (Alpha)"=&FloatFromCurve(
      FloatCurve=DataInterface() {propertyValues {Curve {
         defaultValue=1
         keys [{interpMode=RCIM_Cubic, tangentMode=RCTM_User, time=0, value=1}, {interpMode=RCIM_Cubic, tangentMode=RCTM_User, time=1, value=0}]
        }}}
      Scale Curve=&FNTimeOfDayScalar(Night Scale=2, Morning Scale=1, Evening Scale=2)
     )
    )
    Particles.sparkles color:LinearColor=&MakeLinearColorFromVectorAndFloat(
     "Vector (RGB)"=&MakeVectorFromLinearColorRGB(Color=$Particles.Color)
     "Float (Alpha)"=&FloatFromCurve(
      FloatCurve=DataInterface() {propertyValues {Curve {
         defaultValue=1
         keys [{interpMode=RCIM_Cubic, tangentMode=RCTM_User, time=0, value=1}, {interpMode=RCIM_Cubic, tangentMode=RCTM_User, time=1, value=0}]
        }}}
      Scale Curve=&FNTimeOfDayScalar(Night Scale=3, Morning Scale=2, Day Scale=2, Evening Scale=2)
     )
    )
   ) {isSetParameters=true, module=SetVariables_#4}
  ]
  rendererClasses [&NiagaraSpriteRendererProperties]
  renderers [
   {rendererClass=&NiagaraSpriteRendererProperties}
   {rendererClass=&NiagaraSpriteRendererProperties, rendererIndex=1}
   {rendererClass=&NiagaraSpriteRendererProperties, rendererIndex=2}
   {rendererClass=&NiagaraSpriteRendererProperties, rendererIndex=3}
   {rendererClass=&NiagaraSpriteRendererProperties, rendererIndex=4}
  ]
 }
}
summary {
 systemName=NS_Emote_IlluminateLasso_Sparkles
 emitters [
  {bEnabled=false, emitterName=Sparkles, rendererClasses [&NiagaraSpriteRendererProperties]}
  {bEnabled=false, emitterName=Sparkles001, rendererClasses [&NiagaraSpriteRendererProperties]}
  {emitterName=Sparkles002, rendererClasses [&NiagaraSpriteRendererProperties]}
 ]
 userVariables [
  User.bIsFrontEnd {type=NiagaraBool}
  User.bIsFrontEndPreview {type=NiagaraBool}
 ]
}
