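# Niagara system that kicks up snow around the character during the IlluminateLasso traversal emote.
# Sprite clumps and mist, mesh chunks and a ribbon trail spawn at offsets from a shared socket reader, several split by particle partition.
# Spawn rates switch between traversal and idle values on an owner velocity threshold, with separate front-end values.
# Time-of-day curves tint clumps, mist and trail, while BlendSpace modules handle world-to-local motion.
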NiagaraSystem /BRCosmetics/Animation/Game/MainPlayer/Emotes/IlluminateLasso/FX/NS_Emote_IlluminateLasso_Traversal_Snow
dependencies {
 usedDataInterfaces [
  &NiagaraDataInterfaceVectorCurve
  &NiagaraDataInterfacePlatformSet
  &NiagaraDataInterfaceCurve
  &NiagaraDataInterfaceColorCurve
  &NiagaraDataInterfaceSocketReader
 ]
 usedDynamicInputs [&SimulationPosition, &RandomRangeLinearColor]
 usedModules [
  &self:SystemSpawnScript.SystemScriptSource.SystemScriptGraph.NiagaraNodeAssignment_1.SetVariables_#1
  &SystemState
  &BlendSpaceHelper
  &self:SystemSpawnScript.SystemScriptSource.SystemScriptGraph.NiagaraNodeAssignment_0.SetVariables_#2
  &self:SystemSpawnScript.SystemScriptSource.SystemScriptGraph.NiagaraNodeAssignment_2.SetVariables_#3
  &self:SystemSpawnScript.SystemScriptSource.SystemScriptGraph.NiagaraNodeAssignment_3.SetVariables_#4
 ]
}
emitters {
 ground_clumps_Z_up {
  emitterSpawnScript [/SetVariables(Emitter.FE Particle Spawn Rate=35, Emitter.Max Traversal Speed=150, Emitter.Particle Spawn Rate=20) {isSetParameters=true, module=SetVariables_#5}]
  emitterUpdateScript [
   &EmitterState()
   &SpawnRate(
    SpawnGroup=2
    SpawnRate=&MakeCustomFloatFromBool(
     Bool=&SetBoolByFloatComparison(A=&VectorLength(Vector=$User.OwnerLocalVelocity), B=150)
     True Float=0
     False Float=0
    )
    Use Spawn Probability=false
   ) {enabled=false, module=SpawnRate001}
   /Game/Effects/Niagara/Modules/Logic/UpdateValueStateChange(InputIntParam:NiagaraInt32=&MakeIntFromFloat(Float=0, Conversion Mode=::NewEnumerator1))
   &SpawnRate(
    SpawnRate=&FN_FrontEndFloatModifier(
     Gameplay=&MakeCustomFloatFromBool(
      Bool=&SetBoolByFloatComparison(A=&VectorLength(Vector=$User.OwnerLocalVelocity), B=150)
      True Float=$Emitter.Particle Spawn Rate
      False Float=$Emitter.Idle Spawn Rate
     )
     Front End=$Emitter.FE Particle Spawn Rate
    )
    Use Spawn Probability=false
   )
  ]
  particleSpawnScript [
   &PartitionParticles(Number Of Partitions=1)
   /SetVariables(Particles.InitSpriteRotRate=&RandomRangeFloat(Minimum=-2, Maximum=2)) {isSetParameters=true, module=SetVariables_#6}
   &InitializeParticle(
    Color=$System.Flat snow color
    Lifetime=&RandomRangeFloat(Minimum=0.3, Maximum=0.45)
    Sprite Rotation Angle=&RandomRangeFloat(Minimum=0, Maximum=360)
    Sprite Rotation Mode=::NewEnumerator0
    Sprite Size=&RandomRangeVector2D(Minimum=<20 30>, Maximum=<35 45>)
    Sprite Size Mode=::NewEnumerator4
    Sprite UV Mode=::NewEnumerator2
    UsePositionOffset=false
    Write Material Random=true
   )
   &SocketLocation(
    SamplePosOffset=true
    Sampled Position Offset=<5 0 0>
    Sampled Position Offset Coordinate Space:ENiagaraSkelMeshTransforms=::NewEnumerator2
    Sampled Velocity Coordinate Space:ENiagaraCoordinateSpace=::Local
    Socket Index=$Particles.PartitionParticles.Partition
    Socket Reader=$System.Socket Reader
    Socket Sampling Mode=::NewEnumerator3
    VelocitySpeedLimit=false
   ) {module=SocketLocation001}
   &SphereLocation(
    MaskSpawnGroup=false
    Override Seed=false
    Sphere Coordinate Space=::Simulation
    Sphere Radius=5
    Surface Only=false
    Surface Only Band Thickness=1
    UseNonUniformScale=false
    UseOffset=false
    UseSphereOrientation=false
   )
   /SetVariables(Particles.Lifetime=&MakeCustomFloatFromBool(
     Bool=&SetBoolByFloatComparison(
      A=50
      B=&DistanceBetweenPositions(
       A=&MakeVector(
        X=0
        Y=0
        Z=&MakeFloatFromVector(Vector=Linked() {linkedVariable {name=Particles.SkeletalMeshLocation001.SampledPosition, type=NiagaraPosition}}, Channel=::NewEnumerator2)
       )
       B=&MakeVector(
        X=0
        Y=0
        Z=&MakeFloatFromVector(Vector=Linked() {linkedVariable {name=Particles.SkeletalMeshLocation.SampledPosition, type=NiagaraPosition}}, Channel=::NewEnumerator2)
       )
      )
     )
     True Float=$Particles.Lifetime
     False Float=0
    )) {isSetParameters=true, module=SetVariables_#7}
   &AddVelocity(
    Velocity=&MergeHelper_VectorSwitch(0=<-0.5 -3.5 0>, 1=<0.5 -3.5 0>, 2=<0 -0.1 0>, 3=?, Index=$Particles.PartitionParticles.Partition, Num=::NewEnumerator1)
    Velocity Speed Scale=&RandomRangeFloat(Minimum=25, Maximum=80)
   ) {module=AddVelocity001}
   /SetVariables(Particles.SNOW GROUND:Vector3f=<0 0 1>) {isSetParameters=true}
  ]
  particleUpdateScript [
   &ParticleState()
   &Drag(Drag=&RandomRangeFloat(Minimum=0.8, Maximum=1.2), Use Linear Drag=true, Use Rotational Drag=false)
   &SolveForcesAndVelocity(Clamp Velocity=false, Limit Acceleration=false)
   &ScaleSpriteSizeBySpeed(Sample Scale Factor By Curve=true, Velocity Threshold=150)
   &SpriteRotationRate(Rotation Rate=&Multiply_Float(
     A=$Particles.InitSpriteRotRate
     B=&FloatFromCurve(
      FloatCurve=DataInterface() {propertyValues {Curve {keys [
          {arriveTangent=-4.0107, interpMode=RCIM_Cubic, leaveTangent=-4.0107, tangentMode=RCTM_User, time=-0.0046, value=1.0113}
          {arriveTangent=-0.2496, interpMode=RCIM_Cubic, leaveTangent=-0.2496, tangentMode=RCTM_User, time=0.5, value=0.16}
         ]}}}
      CurveIndex=$Emitter.Age
      Scale Curve=150
     )
    ))
   &ScaleColor(
    Scale Alpha=&FloatFromCurve(
     FloatCurve=DataInterface() {propertyValues {Curve {
        defaultValue=1
        keys [
         {interpMode=RCIM_Cubic, tangentMode=RCTM_User, time=0, value=0}
         {interpMode=RCIM_Cubic, tangentMode=RCTM_User, time=0.5, value=1}
         {tangentMode=RCTM_User, time=1, value=0}
        ]
       }}}
     CurveIndex=Linked() {linkedVariable {name=NPC.FortniteNPC.RemappedTimeOfDay_0-1, type=NiagaraFloat}}
    )
    Scale RGB=&MakeVectorFromLinearColorRGB(Color=&ColorFromCurve(
      ColorCurve=DataInterface() {propertyValues {
        AlphaCurve {keys [<at 0 1 RCIM_Cubic>]}
        BlueCurve {keys [
          <at 0.0043 0.7>
          {arriveTangent=-0.1359, interpMode=RCIM_Cubic, leaveTangent=-0.1359, time=0.1789, value=0.7}
          <at 0.2812 0.6624>
          {arriveTangent=1.2198, interpMode=RCIM_Cubic, leaveTangent=1.2198, time=0.3527, value=1}
          {arriveTangent=-1.0217, interpMode=RCIM_Cubic, leaveTangent=-1.0217, time=0.558, value=1}
          {arriveTangent=-0.9848, interpMode=RCIM_Cubic, leaveTangent=-0.9848, time=0.8029, value=0.54}
          <at 0.8615 0.7011>
          <at 1 0.7>
         ]}
        GreenCurve {keys [
          <at 0.0043 0.4211>
          {arriveTangent=0.9841, interpMode=RCIM_Cubic, leaveTangent=0.9841, time=0.1789, value=0.5785}
          <at 0.2812 0.6936>
          {arriveTangent=1.1069, interpMode=RCIM_Cubic, leaveTangent=1.1069, time=0.3527, value=1}
          {arriveTangent=-0.8536, interpMode=RCIM_Cubic, leaveTangent=-0.8536, time=0.558, value=1}
          {arriveTangent=-1.3917, interpMode=RCIM_Cubic, leaveTangent=-1.3917, time=0.8029, value=0.6157}
          <at 0.8615 0.5776>
          <at 1 0.4138>
         ]}
        RedCurve {keys [
          <at 0.0043 0.3653>
          {arriveTangent=1.9379, interpMode=RCIM_Cubic, leaveTangent=1.9379, time=0.1789, value=0.5542}
          <at 0.2812 0.902>
          {arriveTangent=0.3542, interpMode=RCIM_Cubic, leaveTangent=0.3542, time=0.3527, value=1}
          {arriveTangent=-0.2221, interpMode=RCIM_Cubic, leaveTangent=-0.2221, time=0.558, value=1}
          {arriveTangent=-1.476, interpMode=RCIM_Cubic, leaveTangent=-1.476, time=0.8029, value=0.9}
          <at 0.8615 0.552>
          <at 1 0.3566>
         ]}
       }}
      CurveIndex=Linked() {linkedVariable {name=NPC.FortniteNPC.RemappedTimeOfDay_0-1, type=NiagaraFloat}}
     ))
    ScaleA=false
    ScaleRGB=true
   )
   &ScaleColor(
    Scale Alpha=&FloatFromCurve(FloatCurve=DataInterface() {propertyValues {Curve {
        defaultValue=0
        keys [
         {interpMode=RCIM_Cubic, tangentMode=RCTM_User, time=0, value=0}
         {arriveTangent=-0.131, interpMode=RCIM_Cubic, leaveTangent=-0.131, tangentMode=RCTM_User, time=0.1074, value=0.9545}
         {interpMode=RCIM_Cubic, tangentMode=RCTM_User, time=0.5, value=1}
         {tangentMode=RCTM_User, time=1, value=0}
        ]
       }}})
    ScaleA=true
    ScaleRGB=true
   ) {enabled=false, module=ScaleColor001}
   &DynamicMaterialParameters(
    Index 0 Param 1=$Particles.NormalizedAge
    Module.Index0Float0Write=true
    Module.Index0Float1Write=true
    Module.Index0Float2Write=true
    Module.Index0Float3Write=true
   )
   &BlendSpace(Rotation Contribution=0)
  ]
  rendererClasses [&NiagaraSpriteRendererProperties]
  renderers [{rendererClass=&NiagaraSpriteRendererProperties}]
 }
 lil_chunks {
  emitterSpawnScript [/SetVariables(Emitter.FE Particle Spawn Rate=50, Emitter.Max Traversal Speed=100, Emitter.Particle Spawn Rate=50) {isSetParameters=true}]
  emitterUpdateScript [
   &EmitterState(
    Enable Distance Culling=true
    Life Cycle Mode=::NewEnumerator1
    Loop Delay=&RandomRangeFloat(Minimum=0.25, Maximum=0.8)
    Loop Duration=0.35
    MaxDistance=1150
    Module.bUseMaxDistance=true
    Module.bUseMinDistance=false
    Scalability Mode=::NewEnumerator1
    UseLoopDelay=true
   )
   &SpawnRate(
    Spawn Probability=0.7
    SpawnRate=&FN_FrontEndFloatModifier(
     Gameplay=&MakeCustomFloatFromBool(
      Bool=&SetBoolByFloatComparison(A=&VectorLength(Vector=$User.OwnerLocalVelocity), B=150)
      True Float=$Emitter.Particle Spawn Rate
      False Float=$Emitter.Idle Spawn Rate
     )
     Front End=$Emitter.FE Particle Spawn Rate
    )
    Use Spawn Probability=true
   )
  ]
  particleSpawnScript [
   &PartitionParticles()
   &InitializeParticle(
    Color=&Multiply_LinearColorByFloat(Linear Color=$System.snow chunks color, Float=&FN_FrontEndFloatModifier(Gameplay=1, Front End=1))
    Lifetime=&UniformRangedFloat(Minimum=0.45, Maximum=0.65)
    Mesh Scale Mode=::NewEnumerator1
    Mesh Uniform Scale Max=0.1
    Mesh Uniform Scale Min=0.03
    Sprite Rotation Mode=::NewEnumerator2
    Sprite Size Mode=::NewEnumerator3
    Uniform Sprite Size=&MakeFloatFromVector2d(Vector 2D=&Vector2DFromFloat(Value=&UniformRangedFloat(Minimum=4, Maximum=10)))
    UsePositionOffset=false
    Write Material Random=true
   ) {module=InitializeParticle001}
   &ShapeLocation(MaskSpawnGroup=false, Override Seed=false, Shape Origin=$Output.SocketLocation.SampledPosition, Sphere Radius=2, UseOwnerScale=true)
   &AddVelocity(
    Velocity=&MergeHelper_VectorSwitch(
     0=&RandomRangeVector(Minimum=<0 -0.65 0.2>, Maximum=<-0.2 -2 1>)
     1=&RandomRangeVector(Minimum=<0 -0.7 0.2>, Maximum=<0.2 -2 1>)
     2=?
     3=?
     Index=$Particles.PartitionParticles.Partition
     Num=::NewEnumerator0
    )
    Velocity Speed Scale=200
   )
   &SocketLocation(
    SamplePosOffset=true
    Sampled Position Offset=&MaskVector(Mask Source=1, Mask Parameter=0, Mask=$Particles.PartitionParticles.Partition, Masked Vector=<-8 25 10>, Passthrough Vector=<8 25 10>)
    Sampled Position Offset Coordinate Space:ENiagaraSkelMeshTransforms=::NewEnumerator2
    Socket Index=$Particles.PartitionParticles.Partition
    Socket Reader=$System.Socket Reader
    Socket Sampling Mode=::NewEnumerator3
    VelocitySpeedLimit=false
   )
   &InitialMeshOrientation(Mesh Orientation Mode=::NewEnumerator6, Use Rotation Vector=false)
  ]
  particleUpdateScript [
   &ParticleState()
   &ScaleMeshSize(Scale Factor=&RandomRangeVector(
     Minimum=&VectorFromCurve(
      VectorCurve=DataInterface() {propertyValues {XCurve {keys [<at 0 1>,<at 1 0>]}, YCurve {keys [<at 0 1>,<at 1 0>]}, ZCurve {keys [<at 0 1>,<at 1 0>]}}}
      Scale Curve=<0.1 0.1 0.1>
     )
     Maximum=&VectorFromCurve(
      VectorCurve=DataInterface() {propertyValues {XCurve {keys [<at 0 1>,<at 1 0>]}, YCurve {keys [<at 0 1>,<at 1 0>]}, ZCurve {keys [<at 0 1>,<at 1 0>]}}}
      Scale Curve=<0.3 0.3 0.65>
     )
    )) {module=ScaleMeshSize001}
   &ScaleMeshSize(Scale Factor=&RandomRangeVector(Minimum=<0.1 0.1 0.1>, Maximum=<0.3 0.3 0.8>)) {enabled=false}
   &Drag(Drag=0.5, Use Linear Drag=true, Use Rotational Drag=false)
   &GravityForce(Gravity=&RandomRangeVector(Minimum=<0 0 -250>, Maximum=<1 1 -500>))
   &Collision(
    Age Colliding Particles=false
    Collision Enabled=&IsPlatformSetActive(New Platform Set=DataInterface() {propertyValues {Platforms {qualityLevelMask=-8}}})
    KillOnCollision=false
    Module.Control Rest State Via Surface Normal=false
    Module.EnableMaxCollisionCount=false
    Module.Randomize Collision Normal Vector=false
    Radius Calculation Type=::NewEnumerator1
   ) {enabled=false}
   &Collision(
    Age Colliding Particles=false
    Analytical Collision Plane Position 1:Vector3f=<0 0 -20>
    Analytical Collision Plane Space:ENiagaraCoordinateSpace=::Local
    CPU Collision Type=::NewEnumerator3
    Collision Enabled=&IsPlatformSetActive(New Platform Set=DataInterface() {propertyValues {Platforms {qualityLevelMask=-28}}})
    KillOnCollision=false
    Module.Control Rest State Via Surface Normal=false
    Module.EnableMaxCollisionCount=false
    Module.Randomize Collision Normal Vector=false
    Perform 2 Analytical Collision Queries=false
    Restitution=0
   ) {enabled=false, module=Collision001}
   &ScaleColor(Scale RGB=<0.5 0.5 0.5>, ScaleA=true, ScaleRGB=true) {enabled=false}
   &SpriteRotationRate() {enabled=false}
   &BlendSpace(Rotation Contribution=0)
   &AddRotationalVelocity(Rotation Rate=<0.1 0.1 0.2>)
   &SolveForcesAndVelocity(Clamp Velocity=false, Limit Acceleration=false)
   &ScaleSpriteSizeBySpeed(
    Max Scale Factor=&Vector2DFromFloat(Value=1)
    Min Scale Factor=&Vector2DFromFloat(Value=0.25)
    Velocity Threshold=200
   )
  ]
  rendererClasses [&NiagaraMeshRendererProperties]
  renderers [{rendererClass=&NiagaraMeshRendererProperties, rendererIndex=0}]
 }
 side_feet_clumps {
  emitterSpawnScript [/SetVariables(Emitter.FE Particle Spawn Rate=35, Emitter.Max Traversal Speed=150, Emitter.Particle Spawn Rate=30) {isSetParameters=true, module=SetVariables_#5}]
  emitterUpdateScript [
   &EmitterState()
   &UpdateValueStateChange(InputIntParam:NiagaraInt32=&MakeIntFromFloat(Float=0, Conversion Mode=::NewEnumerator1))
   &SpawnRate(
    SpawnRate=&FN_FrontEndFloatModifier(
     Gameplay=&MakeCustomFloatFromBool(
      Bool=&SetBoolByFloatComparison(A=&VectorLength(Vector=$User.OwnerLocalVelocity), B=150)
      True Float=$Emitter.Particle Spawn Rate
      False Float=$Emitter.Idle Spawn Rate
     )
     Front End=$Emitter.FE Particle Spawn Rate
    )
    Use Spawn Probability=false
   )
  ]
  particleSpawnScript [
   &PartitionParticles()
   /SetVariables(Particles.InitSpriteRotRate=&RandomRangeFloat(Minimum=-2, Maximum=2)) {isSetParameters=true, module=SetVariables_#6}
   &InitializeParticle(
    Color=$System.Flat snow color
    Lifetime=&MergeHelper_FloatSwitch(
     0=&RandomRangeFloat(Minimum=0.35, Maximum=0.5)
     1=&RandomRangeFloat(Minimum=0.35, Maximum=0.5)
     2=&RandomRangeFloat(Minimum=0.2, Maximum=0.35)
     3=?
     Index=$Particles.PartitionParticles.Partition
     Num=::NewEnumerator1
    )
    Sprite Rotation Angle=&RandomRangeFloat(Minimum=0, Maximum=360)
    Sprite Rotation Mode=::NewEnumerator0
    Sprite Size=&MergeHelper_Vector2Switch(
     0=&RandomRangeVector2D(Minimum=<10 25>, Maximum=<20 35>)
     1=&RandomRangeVector2D(Minimum=<10 25>, Maximum=<20 35>)
     2=&RandomRangeVector2D(Minimum=<35 45>, Maximum=<50 50>)
     3=<0 0>
     4=<0 0>
     Index=$Particles.PartitionParticles.Partition
     Num=::NewEnumerator3
    )
    Sprite Size Mode=::NewEnumerator4
    Sprite UV Mode=::NewEnumerator2
    UsePositionOffset=false
    Write Material Random=true
   )
   &SocketLocation(
    SamplePosOffset=true
    Sampled Position Offset=&FN_FrontEndVector3Modifier(
     Gameplay=&MergeHelper_VectorSwitch(0=<-8 4 7>, 1=<8 4 7>, 2=?, 3=?, Index=$Particles.PartitionParticles.Partition, Num=::NewEnumerator0)
     Front End=&MergeHelper_VectorSwitch(0=<-6.5 -5 7>, 1=<6.5 -5 7>, 2=?, 3=?, Index=$Particles.PartitionParticles.Partition, Num=::NewEnumerator0)
    )
    Sampled Position Offset Coordinate Space:ENiagaraSkelMeshTransforms=::NewEnumerator2
    Sampled Velocity Coordinate Space:ENiagaraCoordinateSpace=::Local
    Socket Index=$Particles.PartitionParticles.Partition
    Socket Reader=$System.Socket Reader
    Socket Sampling Mode=::NewEnumerator3
    VelocitySpeedLimit=false
   ) {module=SocketLocation001}
   &SphereLocation(
    MaskSpawnGroup=false
    Non Uniform Scale=<1 1 0.2>
    Override Seed=false
    Sphere Coordinate Space=::Simulation
    Sphere Radius=7
    Surface Only=false
    Surface Only Band Thickness=1
    UseNonUniformScale=true
    UseOffset=false
    UseSphereOrientation=false
   )
   /SetVariables(Particles.Lifetime=&MakeCustomFloatFromBool(
     Bool=&SetBoolByFloatComparison(
      A=50
      B=&DistanceBetweenPositions(
       A=&MakeVector(
        X=0
        Y=0
        Z=&MakeFloatFromVector(Vector=Linked() {linkedVariable {name=Particles.SkeletalMeshLocation001.SampledPosition, type=NiagaraPosition}}, Channel=::NewEnumerator2)
       )
       B=&MakeVector(
        X=0
        Y=0
        Z=&MakeFloatFromVector(Vector=Linked() {linkedVariable {name=Particles.SkeletalMeshLocation.SampledPosition, type=NiagaraPosition}}, Channel=::NewEnumerator2)
       )
      )
     )
     True Float=$Particles.Lifetime
     False Float=0
    )) {isSetParameters=true, module=SetVariables_#7}
   &AddVelocity(
    Velocity=&MergeHelper_VectorSwitch(0=<-0.2 -2.5 0>, 1=<0.2 -2.5 0>, 2=<0 -1.5 0.5>, 3=?, Index=$Particles.PartitionParticles.Partition, Num=::NewEnumerator1)
    Velocity Speed Scale=&RandomRangeFloat(Minimum=25, Maximum=70)
   ) {module=AddVelocity001}
  ]
  particleUpdateScript [
   &ParticleState()
   &Drag(Drag=&RandomRangeFloat(Minimum=0.8, Maximum=1.2), Use Linear Drag=true, Use Rotational Drag=false)
   &SolveForcesAndVelocity(Clamp Velocity=false, Limit Acceleration=false)
   &ScaleSpriteSizeBySpeed(Sample Scale Factor By Curve=true, Velocity Threshold=150)
   &SpriteRotationRate(Rotation Rate=&Multiply_Float(
     A=$Particles.InitSpriteRotRate
     B=&FloatFromCurve(
      FloatCurve=DataInterface() {propertyValues {Curve {keys [
          {arriveTangent=-4.0107, interpMode=RCIM_Cubic, leaveTangent=-4.0107, tangentMode=RCTM_User, time=-0.0046, value=1.0113}
          {arriveTangent=-0.2496, interpMode=RCIM_Cubic, leaveTangent=-0.2496, tangentMode=RCTM_User, time=0.5, value=0.16}
         ]}}}
      CurveIndex=$Emitter.Age
      Scale Curve=150
     )
    ))
   &ScaleColor(
    Scale Alpha=&FloatFromCurve(
     FloatCurve=DataInterface() {propertyValues {Curve {
        defaultValue=1
        keys [
         {interpMode=RCIM_Cubic, tangentMode=RCTM_User, time=0, value=0}
         {interpMode=RCIM_Cubic, tangentMode=RCTM_User, time=0.5, value=1}
         {tangentMode=RCTM_User, time=1, value=0}
        ]
       }}}
     CurveIndex=Linked() {linkedVariable {name=NPC.FortniteNPC.RemappedTimeOfDay_0-1, type=NiagaraFloat}}
    )
    Scale RGB=&MakeVectorFromLinearColorRGB(Color=&ColorFromCurve(
      ColorCurve=DataInterface() {propertyValues {
        AlphaCurve {keys [<at 0 1 RCIM_Cubic>]}
        BlueCurve {keys [
          <at 0.0043 0.7>
          {arriveTangent=-0.1359, interpMode=RCIM_Cubic, leaveTangent=-0.1359, time=0.1789, value=0.7}
          <at 0.2812 0.6624>
          {arriveTangent=1.2198, interpMode=RCIM_Cubic, leaveTangent=1.2198, time=0.3527, value=1}
          {arriveTangent=-1.0217, interpMode=RCIM_Cubic, leaveTangent=-1.0217, time=0.558, value=1}
          {arriveTangent=-0.9848, interpMode=RCIM_Cubic, leaveTangent=-0.9848, time=0.8029, value=0.54}
          <at 0.8615 0.7011>
          <at 1 0.7>
         ]}
        GreenCurve {keys [
          <at 0.0043 0.4211>
          {arriveTangent=0.9841, interpMode=RCIM_Cubic, leaveTangent=0.9841, time=0.1789, value=0.5785}
          <at 0.2812 0.6936>
          {arriveTangent=1.1069, interpMode=RCIM_Cubic, leaveTangent=1.1069, time=0.3527, value=1}
          {arriveTangent=-0.8536, interpMode=RCIM_Cubic, leaveTangent=-0.8536, time=0.558, value=1}
          {arriveTangent=-1.3917, interpMode=RCIM_Cubic, leaveTangent=-1.3917, time=0.8029, value=0.6157}
          <at 0.8615 0.5776>
          <at 1 0.4138>
         ]}
        RedCurve {keys [
          <at 0.0043 0.3653>
          {arriveTangent=1.9379, interpMode=RCIM_Cubic, leaveTangent=1.9379, time=0.1789, value=0.5542}
          <at 0.2812 0.902>
          {arriveTangent=0.3542, interpMode=RCIM_Cubic, leaveTangent=0.3542, time=0.3527, value=1}
          {arriveTangent=-0.2221, interpMode=RCIM_Cubic, leaveTangent=-0.2221, time=0.558, value=1}
          {arriveTangent=-1.476, interpMode=RCIM_Cubic, leaveTangent=-1.476, time=0.8029, value=0.9}
          <at 0.8615 0.552>
          <at 1 0.3566>
         ]}
       }}
      CurveIndex=Linked() {linkedVariable {name=NPC.FortniteNPC.RemappedTimeOfDay_0-1, type=NiagaraFloat}}
     ))
    ScaleA=false
    ScaleRGB=true
   )
   &ScaleColor(
    Scale RGB=&MakeVectorFromLinearColorRGB(Color=&ColorFromCurve(
      ColorCurve=DataInterface() {propertyValues {
        AlphaCurve {keys [<at 0 1 RCIM_Cubic>]}
        BlueCurve {keys [
          <at 0.0043 0.7>
          {arriveTangent=-0.2478, interpMode=RCIM_Cubic, leaveTangent=-0.2478, time=0.1789, value=0.7}
          <at 0.2812 0.6314>
          {arriveTangent=1.3318, interpMode=RCIM_Cubic, leaveTangent=1.3318, time=0.3527, value=1}
          {arriveTangent=-1.0217, interpMode=RCIM_Cubic, leaveTangent=-1.0217, time=0.558, value=1}
          {arriveTangent=-0.9848, interpMode=RCIM_Cubic, leaveTangent=-0.9848, time=0.8029, value=0.54}
          <at 0.8615 0.7011>
          <at 1 0.7>
         ]}
        GreenCurve {keys [
          <at 0.0043 0.35>
          {arriveTangent=1.1435, interpMode=RCIM_Cubic, leaveTangent=1.1435, time=0.1789, value=0.5785}
          <at 0.2812 0.6667>
          {arriveTangent=1.2043, interpMode=RCIM_Cubic, leaveTangent=1.2043, time=0.3527, value=1}
          {arriveTangent=-0.8536, interpMode=RCIM_Cubic, leaveTangent=-0.8536, time=0.558, value=1}
          {arriveTangent=-1.3917, interpMode=RCIM_Cubic, leaveTangent=-1.3917, time=0.8029, value=0.6157}
          <at 0.8615 0.5776>
          <at 1 0.35>
         ]}
        RedCurve {keys [
          <at 0.0043 0.28>
          {arriveTangent=2.246, interpMode=RCIM_Cubic, leaveTangent=2.246, time=0.1789, value=0.5542}
          <at 0.2812 0.902>
          {arriveTangent=0.3542, interpMode=RCIM_Cubic, leaveTangent=0.3542, time=0.3527, value=1}
          {arriveTangent=-0.2221, interpMode=RCIM_Cubic, leaveTangent=-0.2221, time=0.558, value=1}
          {arriveTangent=-1.476, interpMode=RCIM_Cubic, leaveTangent=-1.476, time=0.8029, value=0.9}
          <at 0.8615 0.552>
          <at 1 0.28>
         ]}
       }}
      CurveIndex=Linked() {linkedVariable {name=NPC.FortniteNPC.RemappedTimeOfDay_0-1, type=NiagaraFloat}}
     ))
    ScaleA=false
    ScaleRGB=true
   ) {enabled=false, module=ScaleColor001}
   &DynamicMaterialParameters(
    Index 0 Param 1=$Particles.NormalizedAge
    Module.Index0Float0Write=true
    Module.Index0Float1Write=true
    Module.Index0Float2Write=true
    Module.Index0Float3Write=true
   )
   &BlendSpace("Position Blend (World <-> Local)":NiagaraFloat=&MakeCustomFloatFromBool(
     Bool=&SetBoolByFloatComparison(A=&VectorLength(Vector=$Engine.Owner.Velocity), B=400)
     True Float=1
     False Float=&FloatFromCurve(FloatCurve=DataInterface() {propertyValues {Curve {keys [<at 0 1 RCIM_Cubic>, {arriveTangent=-3.5186, interpMode=RCIM_Cubic, leaveTangent=-3.5186, tangentMode=RCTM_User, time=0.5, value=0}]}}})
     Translation Contribution:NiagaraFloat=1
     Rotation Contribution:NiagaraFloat=1
    ))
  ]
  rendererClasses [&NiagaraSpriteRendererProperties]
  renderers [{rendererClass=&NiagaraSpriteRendererProperties}]
 }
 snow_feet_chunks {
  emitterSpawnScript [/SetVariables(Emitter.FE Particle Spawn Rate=35, Emitter.Max Traversal Speed=150, Emitter.Particle Spawn Rate=30) {isSetParameters=true, module=SetVariables_#8}]
  emitterUpdateScript [
   &EmitterState()
   &SpawnRate(
    SpawnRate=&FN_FrontEndFloatModifier(
     Gameplay=&MakeCustomFloatFromBool(
      Bool=&SetBoolByFloatComparison(A=&VectorLength(Vector=$User.OwnerLocalVelocity), B=150)
      True Float=$Emitter.Particle Spawn Rate
      False Float=$Emitter.Idle Spawn Rate
     )
     Front End=$Emitter.FE Particle Spawn Rate
    )
    Use Spawn Probability=false
   )
  ]
  particleSpawnScript [
   &PartitionParticles()
   &InitializeParticle(
    Color=&Multiply_LinearColorByFloat(Linear Color=$System.snow chunks color, Float=&FN_FrontEndFloatModifier(Gameplay=1, Front End=1))
    Lifetime=&RandomRangeFloat(Minimum=0.2, Maximum=0.45)
    Mesh Scale=&Multiply_VectorByFloat(Vector=&RandomRangeVector(Minimum=<2 0.5 0.5>, Maximum=<4 2 2>), Float=0.1)
    Mesh Scale Mode=::NewEnumerator4
    UsePositionOffset=false
    Write Material Random=true
   )
   &SocketLocation(
    Orientation Sampling=::NewEnumerator1
    SamplePosOffset=true
    Sampled Orientation Coordinate Space:ENiagaraCoordinateSpace=::Local
    Sampled Position Offset=&MaskVector(Mask Source=1, Mask Parameter=0, Mask=$Particles.PartitionParticles.Partition, Masked Vector=<-10 10 -3>, Passthrough Vector=<10 10 -3>)
    Sampled Position Offset Coordinate Space:ENiagaraSkelMeshTransforms=::NewEnumerator2
    Socket Index=$Particles.PartitionParticles.Partition
    Socket Reader=$System.Socket Reader
    Socket Sampling Mode=::NewEnumerator3
    VelocitySpeedLimit=false
   )
   &TorusLocation(Align Torus With Axis=false, Handle Radius=0, Large Radius=8, MaskSpawnGroup=false, Module.Override Randomness=false)
   &AddVelocityFromPoint(Module.Use Falloff Distance=false, Origin Offset=<0 -200 0>, Velocity Strength=-200)
   &AddVelocity(
    Velocity=&MergeHelper_VectorSwitch(
     0=&RandomRangeVector(Minimum=<-0.05 -0.5 0>, Maximum=<-0.2 -0.9 0.1>)
     1=&RandomRangeVector(Minimum=<0.05 -0.65 0>, Maximum=<0.2 -0.85 0.1>)
     2=?
     3=?
     Index=$Particles.PartitionParticles.Partition
     Num=::NewEnumerator0
    )
    Velocity Speed Scale=&RandomRangeFloat(Minimum=200, Maximum=300)
   ) {module=AddVelocity001}
   &AddVelocityInCone(Cone Angle=1, Cone Axis=<0 1 0.5>, Module.Use Velocity Falloff On Cone Axis=false, Velocity Strength=100) {module=AddVelocityInCone001}
   &CameraOffset(Camera Offset Amount=&FN_FrontEndFloatModifier(Gameplay=-5, Front End=-30))
   /SetVariables(Particles.Mesh_Low_Color:LinearColor=&MultiplyLinearColor(Linear Color=<rgba 0.5534 0.6554 0.7135 1>, Scale Factor=&MakeVector4_FromFloat(Float=1))) {isSetParameters=true, module=SetVariables_#9}
  ]
  particleUpdateScript [
   &AccelerationForce(Acceleration=&MakeVector(X=0, Y=0, Z=&FN_FrontEndFloatModifier(Gameplay=0, Front End=0)))
   &GravityForce(Gravity=&RandomRangeVector(Minimum=<0 0 -100>, Maximum=<0 1 -300>))
   &SolveForcesAndVelocity(Clamp Velocity=false, Limit Acceleration=false)
   &ParticleState()
   &ScaleMeshSize(Scale Factor=&Multiply_VectorByFloat(
     Vector=<1 1 1>
     Float=&FloatFromCurve(
      FloatCurve=DataInterface() {propertyValues {Curve {keys [
          {interpMode=RCIM_Cubic, tangentMode=RCTM_User, time=0, value=0}
          <at 0.4 0.8606>
          {arriveTangent=-1.0223, interpMode=RCIM_Cubic, leaveTangent=-1.0223, tangentMode=RCTM_User, time=1, value=0}
         ]}}}
      Scale Curve=&RandomRangeFloat(Minimum=0.2, Maximum=0.5)
     )
    ))
   &ScaleColor(
    Scale Alpha=&FloatFromCurve(
     FloatCurve=DataInterface() {propertyValues {Curve {keys [
         {interpMode=RCIM_Cubic, tangentMode=RCTM_User, time=0, value=0}
         {interpMode=RCIM_Cubic, tangentMode=RCTM_User, time=0.3, value=1}
         {tangentMode=RCTM_User, time=1, value=0}
        ]}}}
     Scale Curve=&RandomRangeFloat(Minimum=0.5, Maximum=1)
    )
    Scale RGB=<0.5 0.5 0.5>
    ScaleA=true
    ScaleRGB=true
   ) {enabled=false}
   &BlendSpace(Rotation Contribution=0)
  ]
  rendererClasses [&NiagaraMeshRendererProperties]
  renderers [{rendererClass=&NiagaraMeshRendererProperties, rendererIndex=0}]
 }
 snow_feet_chunks001 {
  emitterSpawnScript [/SetVariables(Emitter.FE Particle Spawn Rate=35, Emitter.Max Traversal Speed=150, Emitter.Particle Spawn Rate=30) {isSetParameters=true, module=SetVariables_#8}]
  emitterUpdateScript [
   &EmitterState()
   &SpawnRate(
    SpawnRate=&FN_FrontEndFloatModifier(
     Gameplay=&MakeCustomFloatFromBool(
      Bool=&SetBoolByFloatComparison(A=&VectorLength(Vector=$User.OwnerLocalVelocity), B=150)
      True Float=$Emitter.Particle Spawn Rate
      False Float=$Emitter.Idle Spawn Rate
     )
     Front End=$Emitter.FE Particle Spawn Rate
    )
    Use Spawn Probability=false
   )
  ]
  particleSpawnScript [
   &PartitionParticles(Number Of Partitions=1)
   &InitializeParticle(
    Color=&Multiply_LinearColorByFloat(Linear Color=$System.snow chunks color, Float=1)
    Lifetime=&RandomRangeFloat(Minimum=1, Maximum=1)
    Mesh Scale=&Multiply_VectorByFloat(Vector=&RandomRangeVector(Minimum=<2 1.5 0.5>, Maximum=<3 2.5 2>), Float=0.1)
    Mesh Scale Mode=::NewEnumerator4
    UsePositionOffset=false
    Write Material Random=true
   )
   &ShapeLocation(MaskSpawnGroup=false, Override Seed=false, Sphere Radius=5, UseOwnerScale=true)
   &SocketLocation(
    Orientation Sampling=::NewEnumerator1
    SamplePosOffset=true
    Sampled Orientation Coordinate Space:ENiagaraCoordinateSpace=::Local
    Sampled Position Offset=&FN_FrontEndVector3Modifier(Gameplay=<7 50 -5>, Front End=<7 50 -5>)
    Sampled Position Offset Coordinate Space:ENiagaraSkelMeshTransforms=::NewEnumerator2
    Socket Index=$Particles.PartitionParticles.Partition
    Socket Reader=$System.Socket Reader
    Socket Sampling Mode=::NewEnumerator3
    VelocitySpeedLimit=false
   )
   &TorusLocation(Align Torus With Axis=false, Handle Radius=0, Large Radius=8, MaskSpawnGroup=false, Module.Override Randomness=false) {enabled=false}
   &AddVelocityFromPoint(Module.Use Falloff Distance=false, Origin Offset=<0 -200 0>, Velocity Strength=-200)
   &AddVelocity(
    Velocity=&MergeHelper_VectorSwitch(
     0=&RandomRangeVector(Minimum=<-0.05 -0.5 0>, Maximum=<-0.2 -0.9 0.1>)
     1=&RandomRangeVector(Minimum=<0.05 -0.65 0>, Maximum=<0.2 -0.85 0.1>)
     2=?
     3=?
     Index=$Particles.PartitionParticles.Partition
     Num=::NewEnumerator0
    )
    Velocity Speed Scale=&RandomRangeFloat(Minimum=200, Maximum=300)
   ) {module=AddVelocity001}
   &AddVelocityInCone(Cone Angle=1, Cone Axis=<0 1 0.5>, Module.Use Velocity Falloff On Cone Axis=false, Velocity Strength=100) {module=AddVelocityInCone001}
   &CameraOffset(Camera Offset Amount=&FN_FrontEndFloatModifier(Gameplay=-5, Front End=-5))
   /SetVariables(Particles.Mesh_Low_Color:LinearColor=&MultiplyLinearColor(Linear Color=<rgba 0.5534 0.6554 0.7135 1>, Scale Factor=&MakeVector4_FromFloat(Float=1))) {isSetParameters=true, module=SetVariables_#9}
  ]
  particleUpdateScript [
   &AccelerationForce(Acceleration=&MakeVector(X=0, Y=0, Z=&FN_FrontEndFloatModifier(Gameplay=0, Front End=0)))
   &GravityForce(Gravity=&RandomRangeVector(Minimum=<0 0 -100>, Maximum=<0 1 -300>))
   &SolveForcesAndVelocity(Clamp Velocity=false, Limit Acceleration=false)
   &ParticleState()
   &ScaleMeshSize(Scale Factor=&Multiply_VectorByFloat(
     Vector=<1 1 1>
     Float=&FloatFromCurve(
      FloatCurve=DataInterface() {propertyValues {Curve {keys [
          {interpMode=RCIM_Cubic, tangentMode=RCTM_User, time=0, value=0}
          <at 0.144 0.7073>
          {arriveTangent=-1.0223, interpMode=RCIM_Cubic, leaveTangent=-1.0223, tangentMode=RCTM_User, time=0.3, value=0}
         ]}}}
      Scale Curve=&RandomRangeFloat(Minimum=0.55, Maximum=0.75)
     )
    ))
   &ScaleColor(
    Scale Alpha=&FloatFromCurve(
     FloatCurve=DataInterface() {propertyValues {Curve {keys [
         {interpMode=RCIM_Cubic, tangentMode=RCTM_User, time=0, value=0}
         {interpMode=RCIM_Cubic, tangentMode=RCTM_User, time=0.3, value=1}
         {tangentMode=RCTM_User, time=1, value=0}
        ]}}}
     Scale Curve=&RandomRangeFloat(Minimum=0.5, Maximum=1)
    )
    Scale RGB=<0.5 0.5 0.5>
    ScaleA=true
    ScaleRGB=true
   ) {enabled=false}
   &BlendSpace(Rotation Contribution=0)
  ]
  rendererClasses [&NiagaraMeshRendererProperties]
  renderers [
   {rendererClass=&NiagaraMeshRendererProperties, rendererIndex=0}
   {rendererClass=&NiagaraMeshRendererProperties, rendererIndex=1}
  ]
 }
 snow_mist {
  emitterSpawnScript [/SetVariables(Emitter.FE Particle Spawn Rate=35, Emitter.Max Traversal Speed=150, Emitter.Particle Spawn Rate=45) {isSetParameters=true}]
  emitterUpdateScript [
   &EmitterState(Life Cycle Mode=::NewEnumerator1, Loop Duration=1, UseLoopDelay=false)
   &SpawnRate(
    SpawnRate=&FN_FrontEndFloatModifier(
     Gameplay=&MakeCustomFloatFromBool(
      Bool=&SetBoolByFloatComparison(A=&VectorLength(Vector=$User.OwnerLocalVelocity), B=150)
      True Float=$Emitter.Particle Spawn Rate
      False Float=$Emitter.Idle Spawn Rate
     )
     Front End=$Emitter.FE Particle Spawn Rate
    )
    Use Spawn Probability=false
   )
  ]
  particleSpawnScript [
   &InitializeParticle(
    Color=<rgba 0.4492 0.6108 0.8177 0.2>
    Lifetime Max=0.8
    Lifetime Min=0.55
    Lifetime Mode=::NewEnumerator1
    Sprite Rotation Mode=::NewEnumerator2
    Sprite Size Mode=::NewEnumerator1
    Uniform Sprite Size Max=75
    Uniform Sprite Size Min=35
    UsePositionOffset=false
    Write Material Random=true
   )
   &SocketLocation(
    SamplePosOffset=true
    Sampled Position Offset=<0 45 10>
    Sampled Position Offset Coordinate Space:ENiagaraSkelMeshTransforms=::NewEnumerator2
    Sampled Velocity Coordinate Space:ENiagaraCoordinateSpace=::Local
    Socket Reader=$System.Socket Reader
    Socket Sampling Mode=::NewEnumerator3
    VelocitySpeedLimit=false
   ) {module=SocketLocation001}
   &TorusLocation(
    Align Torus With Axis=false
    Handle Radius=0
    Large Radius:NiagaraFloat=&Multiply_Float(A=30, B=&ClampFloat(Float=$User.ForwardDot, Min=0.3))
    MaskSpawnGroup=false
    Module.Override Randomness=false
   )
   &AddVelocityFromPoint(Module.Use Falloff Distance=false, Origin Offset=<0 150 0>)
   &AddVelocity(Velocity=<0 -200 0>)
   &AddVelocityInCone(Cone Angle=10, Cone Axis=<0 -1 0>, Module.Use Velocity Falloff On Cone Axis=false, Velocity Strength=150) {module=AddVelocityInCone001}
   &CameraOffset(Camera Offset Amount=10)
   &DynamicMaterialParameters(
    Index 0 Param 1=1
    Index 0 Param 2=1
    Index 0 Param 3=1
    Index 0 Param 4=1
    Module.Param0WriteEnabled=true
    Module.Param1WriteEnabled=true
    Module.Param2WriteEnabled=true
    Module.Param3WriteEnabled=true
   )
  ]
  particleUpdateScript [
   &AccelerationForce(Acceleration=&MakeVector(X=0, Y=0, Z=300))
   &GravityForce(Gravity=&RandomRangeVector(Minimum=<0 0 -300>, Maximum=<1 1 -500>))
   &Drag(Drag=&RandomRangeFloat(Minimum=1, Maximum=4), Use Linear Drag=true, Use Rotational Drag=true)
   &SolveForcesAndVelocity(Clamp Velocity=false, Limit Acceleration=false)
   &ParticleState()
   &ScaleSpriteSize(
    Scale Sprite Size Mode=::NewEnumerator0
    Uniform Scale Factor=&FloatFromCurve(
     FloatCurve=DataInterface() {propertyValues {Curve {keys [
         {arriveTangent=0.8753, interpMode=RCIM_Cubic, leaveTangent=0.8753, tangentMode=RCTM_User, time=0, value=0.7}
         {arriveTangent=0.0238, interpMode=RCIM_Cubic, leaveTangent=0.0238, tangentMode=RCTM_User, time=1, value=2.5}
        ]}}}
     Scale Curve=0.75
    )
   )
   &ScaleColor(
    Scale Alpha=&Multiply_Float(
     A=&FloatFromCurve(
      FloatCurve=DataInterface() {propertyValues {Curve {keys [
          {interpMode=RCIM_Cubic, tangentMode=RCTM_User, time=0, value=0}
          {interpMode=RCIM_Cubic, leaveTangent=-2.6773, tangentMode=RCTM_Break, time=0.3, value=1}
          {tangentMode=RCTM_User, time=1, value=0}
         ]}}}
      Scale Curve=&RandomRangeFloat(Minimum=0.5, Maximum=1)
     )
     B=&FNTimeOfDayScalar(Night Scale=0.6, Morning Scale=0.7, Evening Scale=0.5)
    )
    ScaleA=true
    ScaleRGB=true
   )
   &SubUVAnimation(SubUV Loop Count=0.3)
   &BlendSpace("Position Blend (World <-> Local)":NiagaraFloat=&MakeCustomFloatFromBool(
     Bool=&SetBoolByFloatComparison(A=&VectorLength(Vector=$Engine.Owner.Velocity), B=400)
     True Float=1
     False Float=&FloatFromCurve(FloatCurve=DataInterface() {propertyValues {Curve {keys [<at 0 1 RCIM_Cubic>, {arriveTangent=-3.5186, interpMode=RCIM_Cubic, leaveTangent=-3.5186, tangentMode=RCTM_User, time=1, value=0}]}}})
     Translation Contribution:NiagaraFloat=1
     Rotation Contribution:NiagaraFloat=1
    ))
  ]
  rendererClasses [&NiagaraSpriteRendererProperties]
  renderers [{rendererClass=&NiagaraSpriteRendererProperties}]
 }
 snowtrail {
  emitterSpawnScript [/SetVariables(Emitter.FE Particle Spawn Rate=50, Emitter.Max Traversal Speed=150, Emitter.Particle Spawn Rate=50) {isSetParameters=true}]
  emitterUpdateScript [
   &EmitterState(Life Cycle Mode=::NewEnumerator1, Loop Behavior=::NewEnumerator1, Loop Delay=0.389, Loop Duration Mode=::NewEnumerator1, UseLoopDelay=true)
   &SpawnRate(
    SpawnRate=&FN_FrontEndFloatModifier(
     Gameplay=&MakeCustomFloatFromBool(
      Bool=&SetBoolByFloatComparison(A=&VectorLength(Vector=$User.OwnerLocalVelocity), B=150)
      True Float=$Emitter.Particle Spawn Rate
      False Float=$Emitter.Idle Spawn Rate
     )
     Front End=$Emitter.FE Particle Spawn Rate
    )
    Use Spawn Probability=false
   )
  ]
  particleSpawnScript [
   &InitializeParticle(
    Color=&FNTimeOfDayColor(
     Night Color=<rgba 0.4594 0.5872 1 0.828>
     Morning Color=<rgba 0.5594 0.7209 1 0.828>
     Day Color=<rgba 0.5094 0.6892 1 0.828>
     Evening Color=<rgba 0.5677 0.7261 1 0.828>
    )
    Lifetime=0.5
    Ribbon Facing Mode=::NewEnumerator1
    Ribbon Facing Vector:Vector3f=$Engine.Owner.SystemZAxis
    Ribbon Width=150
    Ribbon Width Mode=::NewEnumerator1
    UsePositionOffset=false
    Write Material Random=true
   )
   &SocketLocation(
    SamplePosOffset=true
    Sampled Position Offset=<5 40 2.5>
    Sampled Position Offset Coordinate Space:ENiagaraSkelMeshTransforms=::NewEnumerator2
    Socket Reader=$System.Socket Reader
    Socket Sampling Mode=::NewEnumerator3
    VelocitySpeedLimit=false
   )
   &AddVelocity(Velocity=&MakeVector(X=0, Y=-800, Z=0))
  ]
  particleUpdateScript [
   &ParticleState()
   &ScaleColor(
    Scale Alpha=&FloatFromCurve(
     FloatCurve=DataInterface() {propertyValues {Curve {keys [
         {interpMode=RCIM_Cubic, tangentMode=RCTM_Break, time=0, value=0}
         {arriveTangent=1.6363, interpMode=RCIM_Cubic, leaveTangent=1.6363, time=0.1824, value=0.859}
         {arriveTangent=-1.6833, interpMode=RCIM_Cubic, leaveTangent=-1.6833, time=0.4871, value=0.797}
         {interpMode=RCIM_Cubic, tangentMode=RCTM_Break, time=0.7227, value=-0.0504}
        ]}}}
     Scale Curve=0.7
    )
    ScaleA=true
    ScaleRGB=true
   )
   &ScaleRibbonWidth(Ribbon Width Scale=&FloatFromCurve(
     FloatCurve=DataInterface() {propertyValues {Curve {keys [
         <at 0.0009 0.3 RCIM_Cubic>
         {interpMode=RCIM_Cubic, tangentMode=RCTM_Break, time=0.2, value=1}
         {arriveTangent=-1, interpMode=RCIM_Cubic, leaveTangent=-1, time=0.7536, value=0.9082}
         {arriveTangent=-2.4767, interpMode=RCIM_Cubic, tangentMode=RCTM_Break, time=1, value=0.2}
        ]}}}
     Scale Curve=0.7
    ))
   &Drag(Drag=3, Use Linear Drag=true, Use Rotational Drag=true)
   &GravityForce(Gravity=<-500 0 0>)
   &SolveForcesAndVelocity(Clamp Velocity=false, Limit Acceleration=false)
   &BlendSpace(Rotation Contribution=0)
  ]
  rendererClasses [&NiagaraRibbonRendererProperties]
  renderers [{rendererClass=&NiagaraRibbonRendererProperties}]
 }
}
summary {
 systemName=NS_Emote_IlluminateLasso_Traversal_Snow
 emitters [
  {emitterName=lil_chunks, rendererClasses [&NiagaraMeshRendererProperties]}
  {emitterName=snowtrail, rendererClasses [&NiagaraRibbonRendererProperties]}
  {emitterName=snow_feet_chunks, rendererClasses [&NiagaraMeshRendererProperties]}
  {emitterName=side_feet_clumps, rendererClasses [&NiagaraSpriteRendererProperties]}
  {emitterName=ground_clumps_Z_up, rendererClasses [&NiagaraSpriteRendererProperties]}
  {emitterName=snow_feet_chunks001, rendererClasses [&NiagaraMeshRendererProperties]}
  {emitterName=snow_mist, rendererClasses [&NiagaraSpriteRendererProperties]}
 ]
 userVariables [
  User.bIsFrontEnd {type=NiagaraBool}
  User.bIsFrontEndPreview {type=NiagaraBool}
  User.bIsInWater {type=NiagaraBool}
  User.BurstIndex {type=NiagaraFloat, defaultValue=0, description=""}
  User.ForwardDot {type=NiagaraFloat, defaultValue=0, description=""}
  User.FrontEnd {type=NiagaraFloat, defaultValue=1, description=""}
  User.OwnerLocalVelocity {type=Vector3f, description="", defaultValue {struct=Vector3f, value=<0 0 0>}}
  User.RibbonIDOffset {type=NiagaraInt32}
  User.RightDot {type=NiagaraFloat, defaultValue=0, description=""}
  User.SkeletalMesh {
   type=NiagaraDataInterfaceSkeletalMesh
   defaultValue {
    struct=NiagaraExt_VariableValue_DataInterface
    value {dataInterface=&self:NiagaraDataInterfaceSkeletalMesh_0}
   }
  }
  User.SpawnRateCoolSlice {type=NiagaraFloat, defaultValue=0, description=""}
  User.VelocityMult {type=NiagaraFloat, defaultValue=2, description=""}
 ]
}
